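# Post-pose Anim Blueprint for the GroovyReader female player body, correcting back belt bones and simulating rope bones.
# It copies the parent mesh pose through two Control Rigs, then chained cached stages apply interpolated additive bone offsets per movement state.
# AnimDynamics then swings the ropes relative to the feet, followed by belt Constraint and Look At nodes.
# EventGraph derives alphas from pawn state, updates gravity override and drives WPO cloth-sim material parameters.
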AnimBlueprint /BRCosmetics/Characters/Player/Female/Medium/Bodies/F_MED_GroovyReader/Meshes/F_MED_GroovyReader_AnimBP
target_skeleton=&/F_MED_GroovyReader_Skeleton
graphs [
 AnimGraph {nodes [
   n1 = Get GravityOverride()
   n2 = Anim Dynamics - Bone: dyn_rope_r_01!(
    ComponentPose=Local To Component(LocalPose=Use cached pose 'ExtraPoses Fix Cache'())
    Alpha=1
    GravityOverride:Vector=@n1
    ExternalForce="0,0,0"
   ) -> (Pose) {
    angularDampingOverride=1
    angularSpringConstant=20000
    bAngularSpring=true
    bOverrideAngularDamping=true
    bUseGravityOverride=true
    gravityScale=0.35
    simulationSpace=BoneRelative
    alphaBoolBlend {blendOption=Linear}
    boundBone {boneName=dyn_rope_r_01}
    componentLinearAccScale=<1 1 1>
    componentLinearVelScale=<1 1 1>
    gravityOverride {z=-980}
    physicsBodyDefinitions [{
      boundBone {boneName=dyn_rope_r_01}
      boxExtents=<10 10 10>
      constraintSetup {angularLimitsMax=<0 25 0>, angularLimitsMin=<0 -15 0>, angularTarget=<1 0 0>}
      localJointOffset=<1 0 0>
     }]
    relativeSpaceBone {boneName=foot_r}
   }
   n3 = Anim Dynamics - Bone: dyn_rope_r_02!(ComponentPose=@n2, Alpha=1, GravityOverride:Vector=@n1, ExternalForce="0,0,0") -> (Pose) {
    angularDampingOverride=1
    angularSpringConstant=20000
    bAngularSpring=true
    bOverrideAngularDamping=true
    bUseGravityOverride=true
    gravityScale=0.35
    simulationSpace=BoneRelative
    alphaBoolBlend {blendOption=Linear}
    boundBone {boneName=dyn_rope_r_02}
    componentLinearAccScale=<1 1 1>
    componentLinearVelScale=<1 1 1>
    gravityOverride {z=-980}
    physicsBodyDefinitions [{
      boundBone {boneName=dyn_rope_r_02}
      boxExtents=<10 10 10>
      constraintSetup {angularLimitsMax=<0 25 0>, angularLimitsMin=<0 -15 0>, angularTarget=<1 0 0>}
      localJointOffset=<1 0 0>
     }]
    relativeSpaceBone {boneName=foot_l}
   }
   n4 = Anim Dynamics - Bone: dyn_rope_r_03!(ComponentPose=@n3, Alpha=1, GravityOverride:Vector=@n1, ExternalForce="0,0,0") -> (Pose) {
    angularDampingOverride=1
    angularSpringConstant=20000
    bAngularSpring=true
    bOverrideAngularDamping=true
    bUseGravityOverride=true
    gravityScale=0.35
    simulationSpace=BoneRelative
    alphaBoolBlend {blendOption=Linear}
    boundBone {boneName=dyn_rope_r_03}
    componentLinearAccScale=<1 1 1>
    componentLinearVelScale=<1 1 1>
    gravityOverride {z=-980}
    physicsBodyDefinitions [{
      boundBone {boneName=dyn_rope_r_03}
      boxExtents=<10 10 10>
      constraintSetup {angularLimitsMax=<0 25 0>, angularLimitsMin=<0 -15 0>, angularTarget=<1 0 0>}
      localJointOffset=<1 0 0>
     }]
    relativeSpaceBone {boneName=foot_r}
   }
   n5 = Anim Dynamics - Bone: dyn_rope_r_04!(ComponentPose=@n4, Alpha=1, GravityOverride:Vector=@n1, ExternalForce="0,0,0") -> (Pose) {
    angularDampingOverride=1
    angularSpringConstant=20000
    bAngularSpring=true
    bOverrideAngularDamping=true
    bUseGravityOverride=true
    gravityScale=0.35
    simulationSpace=BoneRelative
    alphaBoolBlend {blendOption=Linear}
    boundBone {boneName=dyn_rope_r_04}
    componentLinearAccScale=<1 1 1>
    componentLinearVelScale=<1 1 1>
    gravityOverride {z=-980}
    physicsBodyDefinitions [{
      boundBone {boneName=dyn_rope_r_04}
      boxExtents=<10 10 10>
      constraintSetup {angularLimitsMax=<0 25 0>, angularLimitsMin=<0 -15 0>, angularTarget=<1 0 0>}
      localJointOffset=<1 0 0>
     }]
    relativeSpaceBone {boneName=foot_l}
   }
   n6 = Anim Dynamics - Bone: dyn_rope_l_01!(ComponentPose=@n5, Alpha=1, GravityOverride:Vector=@n1, ExternalForce="0,0,0") -> (Pose) {
    angularDampingOverride=1
    angularSpringConstant=20000
    bAngularSpring=true
    bOverrideAngularDamping=true
    bUseGravityOverride=true
    gravityScale=0.35
    simulationSpace=BoneRelative
    alphaBoolBlend {blendOption=Linear}
    boundBone {boneName=dyn_rope_l_01}
    componentLinearAccScale=<1 1 1>
    componentLinearVelScale=<1 1 1>
    gravityOverride {z=-980}
    physicsBodyDefinitions [{
      boundBone {boneName=dyn_rope_l_01}
      boxExtents=<10 10 10>
      constraintSetup {angularLimitsMax=<0 25 0>, angularLimitsMin=<0 -15 0>, angularTarget=<1 0 0>}
      localJointOffset=<1 0 0>
     }]
    relativeSpaceBone {boneName=foot_r}
   }
   n7 = Anim Dynamics - Bone: dyn_rope_l_02!(ComponentPose=@n6, Alpha=1, GravityOverride:Vector=@n1, ExternalForce="0,0,0") -> (Pose) {
    angularDampingOverride=1
    angularSpringConstant=20000
    bAngularSpring=true
    bOverrideAngularDamping=true
    bUseGravityOverride=true
    gravityScale=0.35
    simulationSpace=BoneRelative
    alphaBoolBlend {blendOption=Linear}
    boundBone {boneName=dyn_rope_l_02}
    componentLinearAccScale=<1 1 1>
    componentLinearVelScale=<1 1 1>
    gravityOverride {z=-980}
    physicsBodyDefinitions [{
      boundBone {boneName=dyn_rope_l_02}
      boxExtents=<10 10 10>
      constraintSetup {angularLimitsMax=<0 25 0>, angularLimitsMin=<0 -15 0>, angularTarget=<1 0 0>}
      localJointOffset=<1 0 0>
     }]
    relativeSpaceBone {boneName=foot_l}
   }
   n8 = Anim Dynamics - Bone: dyn_rope_l_03!(ComponentPose=@n7, Alpha=1, GravityOverride:Vector=@n1, ExternalForce="0,0,0") -> (Pose) {
    angularDampingOverride=1
    angularSpringConstant=20000
    bAngularSpring=true
    bOverrideAngularDamping=true
    bUseGravityOverride=true
    gravityScale=0.35
    simulationSpace=BoneRelative
    alphaBoolBlend {blendOption=Linear}
    boundBone {boneName=dyn_rope_l_03}
    componentLinearAccScale=<1 1 1>
    componentLinearVelScale=<1 1 1>
    gravityOverride {z=-980}
    physicsBodyDefinitions [{
      boundBone {boneName=dyn_rope_l_03}
      boxExtents=<10 10 10>
      constraintSetup {angularLimitsMax=<0 25 0>, angularLimitsMin=<0 -15 0>, angularTarget=<1 0 0>}
      localJointOffset=<1 0 0>
     }]
    relativeSpaceBone {boneName=foot_r}
   }
   n9 = Anim Dynamics - Bone: dyn_rope_l_04!(ComponentPose=@n8, Alpha=1, GravityOverride:Vector=@n1, ExternalForce="0,0,0") -> (Pose) {
    angularDampingOverride=1
    angularSpringConstant=20000
    bAngularSpring=true
    bOverrideAngularDamping=true
    bUseGravityOverride=true
    gravityScale=0.35
    simulationSpace=BoneRelative
    alphaBoolBlend {blendOption=Linear}
    boundBone {boneName=dyn_rope_l_04}
    componentLinearAccScale=<1 1 1>
    componentLinearVelScale=<1 1 1>
    gravityOverride {z=-980}
    physicsBodyDefinitions [{
      boundBone {boneName=dyn_rope_l_04}
      boxExtents=<10 10 10>
      constraintSetup {angularLimitsMax=<0 25 0>, angularLimitsMin=<0 -15 0>, angularTarget=<1 0 0>}
      localJointOffset=<1 0 0>
     }]
    relativeSpaceBone {boneName=foot_l}
   }
   n10 = "Constraint - dyn_beltBack_01_r (P)"(
    ComponentPose=@n9
    Alpha=float * float(A=Get Alpha_NoSprinting() -> (Alpha_NoSprinting:double), B=0.18) {is=title K2Node_PromotableOperator}
    ConstraintWeights_0=1
   ) {
    alphaBoolBlend {blendOption=Linear}
    boneToModify {boneName=dyn_beltBack_01_r}
    constraintSetup [{targetBone {boneName=clavicle_r}}]
    constraintWeights [1]
   }
   n11 = Look At - Bone: dyn_beltBack_02b_l(ComponentPose=@n10, Alpha=1) {
    bUseLookUpAxis=true
    interpolationType=Cubic
    alphaBoolBlend {blendOption=Linear}
    boneToModify {boneName=dyn_beltBack_02b_l}
    lookAtLocation=<11.6664 0 21.9235>
    lookAtTarget {boneReference {boneName=dyn_beltBack_01_l}}
    lookAt_Axis {axis=<1 0>}
    lookUp_Axis {axis {x=0, z=1}}
   }
   Output Pose(Result=Component To Local(ComponentPose=@n11)) {layerGroup=None}
   n12 = Control Rig(Source=Copy Pose From Mesh() {bUseAttachedParent=true}) {
    alphaBoolBlend {blendOption=Linear}
    alphaScaleBiasClamp {bInterpResult=true, interpSpeedDecreasing=5, interpSpeedIncreasing=2}
    controlRigAssetReference {blueprintRigClass=&/F_MED_GroovyReader_ControlRig.F_MED_GroovyReader_ControlRig_C}
   }
   n13 = Control Rig!(Source=@n12, Alpha, IsSprintingSkyDive:bool=Get Is Sprinting Sky Dive() -> (Is Sprinting Sky Dive:bool)) -> (Pose) {
    alphaBoolBlend {blendOption=Linear}
    alphaScaleBiasClamp {interpSpeedDecreasing=2, interpSpeedIncreasing=2}
    controlRigAssetReference {blueprintRigClass=&/F_MED_GroovyReader_shoulder_CtrlRig.F_MED_GroovyReader_shoulder_CtrlRig_C}
   }
   n14 = Get Alpha_isSprintingNoCrouch() -> (Alpha_isSprintingNoCrouch:double)
   n15 = "Transform (Modify) Bone - Bone: base_beltBack_02_r"(ComponentPose=Local To Component(LocalPose=@n13), Alpha=@n14, Translation="0,0,0.3", Rotation="0,0,0", Scale="1,1,1") {
    rotationMode=BMM_Additive
    rotationSpace=BCS_BoneSpace
    translationMode=BMM_Additive
    translationSpace=BCS_BoneSpace
    alphaBoolBlend {blendOption=Linear}
    alphaScaleBiasClamp {bInterpResult=true, interpSpeedDecreasing=5, interpSpeedIncreasing=5}
    boneToModify {boneName=base_beltBack_02_r}
    translation {z=0.3}
   }
   n16 = "Transform (Modify) Bone - Bone: dyn_beltBack_01_r"(ComponentPose=@n15, Alpha=@n14, Translation="0,0,1.5", Rotation="0,0,0", Scale="1,1,1") {
    rotationMode=BMM_Additive
    rotationSpace=BCS_BoneSpace
    translationMode=BMM_Additive
    translationSpace=BCS_BoneSpace
    alphaBoolBlend {blendOption=Linear}
    alphaScaleBiasClamp {bInterpResult=true, interpSpeedDecreasing=5, interpSpeedIncreasing=5}
    boneToModify {boneName=dyn_beltBack_01_r}
    translation {z=1.5}
   }
   n17 = Get Alpha_Sprinting() -> (Alpha_Sprinting:double)
   n18 = "Transform (Modify) Bone - Bone: dyn_beltBack_02b_l"(ComponentPose=@n16, Alpha=@n17, Translation="0,0,1", Rotation="0,0,0", Scale="1,1,1") {
    rotationMode=BMM_Additive
    rotationSpace=BCS_BoneSpace
    translationMode=BMM_Additive
    translationSpace=BCS_BoneSpace
    alphaBoolBlend {blendOption=Linear}
    alphaScaleBiasClamp {bInterpResult=true, interpSpeedDecreasing=5, interpSpeedIncreasing=5}
    boneToModify {boneName=dyn_beltBack_02b_l}
    translation {z=1}
   }
   n19 = "Transform (Modify) Bone - Bone: dyn_beltBack_0_r"(ComponentPose=@n18, Alpha=@n17, Translation="0,0,2", Rotation="0,0,0", Scale="1,1,1") {
    rotationMode=BMM_Additive
    rotationSpace=BCS_BoneSpace
    translationMode=BMM_Additive
    translationSpace=BCS_BoneSpace
    alphaBoolBlend {blendOption=Linear}
    alphaScaleBiasClamp {bInterpResult=true, interpSpeedDecreasing=5, interpSpeedIncreasing=5}
    boneToModify {boneName=dyn_beltBack_0_r}
    translation {z=2}
   }
   Save cached pose 'IsSprinting Fix Cache'(Pose=Component To Local(ComponentPose=@n19)) {cacheName=IsSprinting Fix Cache}
   n20 = "Transform (Modify) Bone - Bone: dyn_beltBack_01_r"(
    ComponentPose=Local To Component(LocalPose=Use cached pose 'IsSprinting Fix Cache'())
    Alpha=Get Alpha_isTargeting() -> (Alpha_isTargeting:double)
    Translation="0,-1.2,0"
    Rotation="0,0,0"
    Scale="1,1,1"
   ) {
    rotationMode=BMM_Additive
    rotationSpace=BCS_BoneSpace
    translationMode=BMM_Additive
    translationSpace=BCS_BoneSpace
    alphaBoolBlend {blendOption=Linear}
    alphaScaleBiasClamp {bInterpResult=true, interpSpeedDecreasing=5, interpSpeedIncreasing=5}
    boneToModify {boneName=dyn_beltBack_01_r}
    translation {y=-1.2}
   }
   Save cached pose 'IsTargeting Fix cache'(Pose=Component To Local(ComponentPose=@n20)) {cacheName=IsTargeting Fix cache}
   n21 = Get Alpha_skydiveMode() -> (Alpha_skydiveMode:float)
   n22 = "Transform (Modify) Bone - Bone: dyn_beltBack_0_l"(
    ComponentPose=Local To Component(LocalPose=Use cached pose 'IsTargeting Fix cache'())
    Alpha=@n21
    Translation="0,0,1"
    Rotation="0,0,0"
    Scale="1,1,1"
   ) {
    rotationMode=BMM_Additive
    rotationSpace=BCS_BoneSpace
    translationMode=BMM_Additive
    translationSpace=BCS_BoneSpace
    alphaBoolBlend {blendOption=Linear}
    alphaScaleBiasClamp {bInterpResult=true, interpSpeedDecreasing=5, interpSpeedIncreasing=5}
    boneToModify {boneName=dyn_beltBack_0_l}
    translation {z=1}
   }
   n23 = "Transform (Modify) Bone - Bone: dyn_beltBack_0_r"(ComponentPose=@n22, Alpha=@n21, Translation="0,0,1", Rotation="0,0,0", Scale="1,1,1") {
    rotationMode=BMM_Additive
    rotationSpace=BCS_BoneSpace
    translationMode=BMM_Additive
    translationSpace=BCS_BoneSpace
    alphaBoolBlend {blendOption=Linear}
    alphaScaleBiasClamp {bInterpResult=true, interpSpeedDecreasing=5, interpSpeedIncreasing=5}
    boneToModify {boneName=dyn_beltBack_0_r}
    translation {z=1}
   }
   n24 = "Transform (Modify) Bone - Bone: dyn_beltBack_01_r"(ComponentPose=@n23, Alpha=@n21, Translation="0,0,2.5", Rotation="0,0,0", Scale="1,1,1") {
    rotationMode=BMM_Additive
    rotationSpace=BCS_BoneSpace
    translationMode=BMM_Additive
    translationSpace=BCS_BoneSpace
    alphaBoolBlend {blendOption=Linear}
    alphaScaleBiasClamp {bInterpResult=true, interpSpeedDecreasing=5, interpSpeedIncreasing=5}
    boneToModify {boneName=dyn_beltBack_01_r}
    translation {z=2.5}
   }
   Save cached pose 'Skydive Mode Fix Cache'(Pose=Component To Local(ComponentPose=@n24)) {cacheName=Skydive Mode Fix Cache}
   n25 = Get Alpha_jumping() -> (Alpha_jumping:float)
   n26 = "Transform (Modify) Bone - Bone: base_beltBack_02_l"(
    ComponentPose=Local To Component(LocalPose=Use cached pose 'Skydive Mode Fix Cache'())
    Alpha=@n25
    Translation="0,0,0.25"
    Rotation="0,0,0"
    Scale="1,1,1"
   ) {
    rotationMode=BMM_Additive
    rotationSpace=BCS_BoneSpace
    translationMode=BMM_Additive
    translationSpace=BCS_BoneSpace
    alphaBoolBlend {blendOption=Linear}
    alphaScaleBiasClamp {bInterpResult=true, interpSpeedDecreasing=15, interpSpeedIncreasing=5}
    boneToModify {boneName=base_beltBack_02_l}
    translation {z=0.25}
   }
   n27 = "Transform (Modify) Bone - Bone: dyn_beltBack_01_l"(ComponentPose=@n26, Alpha=@n25, Translation="0,-1,0", Rotation="0,0,0", Scale="1,1,1") {
    rotationMode=BMM_Additive
    rotationSpace=BCS_BoneSpace
    translationMode=BMM_Additive
    translationSpace=BCS_BoneSpace
    alphaBoolBlend {blendOption=Linear}
    alphaScaleBiasClamp {bInterpResult=true, interpSpeedDecreasing=15, interpSpeedIncreasing=5}
    boneToModify {boneName=dyn_beltBack_01_l}
    translation {y=-1}
   }
   n28 = "Transform (Modify) Bone - Bone: dyn_beltBack_0_l"(ComponentPose=@n27, Alpha=@n25, Translation="0,0,8", Rotation="0,0,0", Scale="1,1,1") {
    rotationMode=BMM_Additive
    rotationSpace=BCS_BoneSpace
    translationMode=BMM_Additive
    translationSpace=BCS_BoneSpace
    alphaBoolBlend {blendOption=Linear}
    alphaScaleBiasClamp {bInterpResult=true, interpSpeedDecreasing=15, interpSpeedIncreasing=5}
    boneToModify {boneName=dyn_beltBack_0_l}
    translation {z=8}
   }
   n29 = "Transform (Modify) Bone - Bone: dyn_beltBack_0_r"(ComponentPose=@n28, Alpha=@n25, Translation="0,-4.5,0", Rotation="0,0,0", Scale="1,1,1") {
    rotationMode=BMM_Additive
    rotationSpace=BCS_BoneSpace
    translationMode=BMM_Additive
    translationSpace=BCS_BoneSpace
    alphaBoolBlend {blendOption=Linear}
    alphaScaleBiasClamp {bInterpResult=true, interpSpeedDecreasing=15, interpSpeedIncreasing=5}
    boneToModify {boneName=dyn_beltBack_0_r}
    translation {y=-4.5}
   }
   Save cached pose 'Jumping Fix Cache'(Pose=Component To Local(ComponentPose=@n29)) {cacheName=Jumping Fix Cache}
   n30 = Get Alpha_openParachused() -> (Alpha_openParachused:float)
   n31 = "Transform (Modify) Bone - Bone: base_beltBack_02_l"(
    ComponentPose=Local To Component(LocalPose=Use cached pose 'Jumping Fix Cache'())
    Alpha=@n30
    Translation="0,2,0"
    Rotation="0,-25,0"
    Scale="1,1,1"
   ) {
    rotationMode=BMM_Additive
    rotationSpace=BCS_BoneSpace
    translationMode=BMM_Additive
    translationSpace=BCS_BoneSpace
    alphaBoolBlend {blendOption=Linear}
    alphaScaleBiasClamp {bInterpResult=true, interpSpeedDecreasing=5, interpSpeedIncreasing=5}
    boneToModify {boneName=base_beltBack_02_l}
    rotation {yaw=-25}
    translation {y=2}
   }
   n32 = "Transform (Modify) Bone - Bone: base_beltBack_02_r"(ComponentPose=@n31, Alpha=@n30, Translation="0,1,0", Rotation="0,25,0", Scale="1,1,1") {
    rotationMode=BMM_Additive
    rotationSpace=BCS_BoneSpace
    translationMode=BMM_Additive
    translationSpace=BCS_BoneSpace
    alphaBoolBlend {blendOption=Linear}
    alphaScaleBiasClamp {bInterpResult=true, interpSpeedDecreasing=5, interpSpeedIncreasing=5}
    boneToModify {boneName=base_beltBack_02_r}
    rotation {yaw=25}
    translation {y=1}
   }
   n33 = "Transform (Modify) Bone - Bone: dyn_beltBack_01_r"(ComponentPose=@n32, Alpha=@n30, Translation="0,3,5", Rotation="0,0,0", Scale="1,1,1") {
    rotationMode=BMM_Additive
    rotationSpace=BCS_BoneSpace
    translationMode=BMM_Additive
    translationSpace=BCS_BoneSpace
    alphaBoolBlend {blendOption=Linear}
    alphaScaleBiasClamp {bInterpResult=true, interpSpeedDecreasing=5, interpSpeedIncreasing=5}
    boneToModify {boneName=dyn_beltBack_01_r}
    translation {y=3, z=5}
   }
   n34 = "Transform (Modify) Bone - Bone: dyn_beltBack_01_l"(ComponentPose=@n33, Alpha=@n30, Translation="0,-3,6", Rotation="0,0,0", Scale="1,1,1") {
    rotationMode=BMM_Additive
    rotationSpace=BCS_BoneSpace
    translationMode=BMM_Additive
    translationSpace=BCS_BoneSpace
    alphaBoolBlend {blendOption=Linear}
    alphaScaleBiasClamp {bInterpResult=true, interpSpeedDecreasing=5, interpSpeedIncreasing=5}
    boneToModify {boneName=dyn_beltBack_01_l}
    translation {y=-3, z=6}
   }
   Save cached pose 'Open Parachute Fix Cache'(Pose=Component To Local(ComponentPose=@n34)) {cacheName=Open Parachute Fix Cache}
   n35 = Get Alpha_SkyDiveFloating() -> (Alpha_SkyDiveFloating:float)
   n36 = "Transform (Modify) Bone - Bone: dyn_beltBack_0_r"(
    ComponentPose=Local To Component(LocalPose=Use cached pose 'Open Parachute Fix Cache'())
    Alpha=@n35
    Translation="0,1,2"
    Rotation="0,0,0"
    Scale="1,1,1"
   ) {
    rotationMode=BMM_Additive
    rotationSpace=BCS_BoneSpace
    translationMode=BMM_Additive
    translationSpace=BCS_BoneSpace
    alphaBoolBlend {blendOption=Linear}
    alphaScaleBiasClamp {bInterpResult=true, interpSpeedDecreasing=5, interpSpeedIncreasing=5}
    boneToModify {boneName=dyn_beltBack_0_r}
    translation {y=1, z=2}
   }
   n37 = "Transform (Modify) Bone - Bone: dyn_beltBack_01_r"(ComponentPose=@n36, Alpha=@n35, Translation="0,2,1", Rotation="0,0,0", Scale="1,1,1") {
    rotationMode=BMM_Additive
    rotationSpace=BCS_BoneSpace
    translationMode=BMM_Additive
    translationSpace=BCS_BoneSpace
    alphaBoolBlend {blendOption=Linear}
    alphaScaleBiasClamp {bInterpResult=true, interpSpeedDecreasing=5, interpSpeedIncreasing=5}
    boneToModify {boneName=dyn_beltBack_01_r}
    translation {y=2, z=1}
   }
   n38 = "Transform (Modify) Bone - Bone: dyn_beltBack_0_l"(ComponentPose=@n37, Alpha=@n35, Translation="-0.017,-0.38,5", Rotation="0,0,0", Scale="1,1,1") {
    rotationMode=BMM_Additive
    rotationSpace=BCS_BoneSpace
    translationMode=BMM_Additive
    translationSpace=BCS_BoneSpace
    alphaBoolBlend {blendOption=Linear}
    alphaScaleBiasClamp {bInterpResult=true, interpSpeedDecreasing=5, interpSpeedIncreasing=5}
    boneToModify {boneName=dyn_beltBack_0_l}
    translation=<-0.017 -0.38 5>
   }
   n39 = "Transform (Modify) Bone - Bone: dyn_beltBack_01_l"(ComponentPose=@n38, Alpha=@n35, Translation="0,0,1", Rotation="0,0,0", Scale="1,1,1") {
    rotationMode=BMM_Additive
    rotationSpace=BCS_BoneSpace
    translationMode=BMM_Additive
    translationSpace=BCS_BoneSpace
    alphaBoolBlend {blendOption=Linear}
    alphaScaleBiasClamp {bInterpResult=true, interpSpeedDecreasing=5, interpSpeedIncreasing=5}
    boneToModify {boneName=dyn_beltBack_01_l}
    translation {z=1}
   }
   n40 = "Transform (Modify) Bone - Bone: base_beltBack_02_l"(ComponentPose=@n39, Alpha=@n35, Translation="-0.184,1.688,0", Rotation="6.845,12.128,2.443", Scale="1,1,1") {
    rotationMode=BMM_Additive
    rotationSpace=BCS_BoneSpace
    translationMode=BMM_Additive
    translationSpace=BCS_BoneSpace
    alphaBoolBlend {blendOption=Linear}
    alphaScaleBiasClamp {bInterpResult=true, interpSpeedDecreasing=5, interpSpeedIncreasing=5}
    boneToModify {boneName=base_beltBack_02_l}
    rotation=<rot 6.845 12.128 2.443>
    translation=<-0.184 1.688>
   }
   Save cached pose 'SkyDiveFloating Fix Cache'(Pose=Component To Local(ComponentPose=@n40)) {cacheName=SkyDiveFloating Fix Cache}
   n41 = Get Backward() -> (Backward:double)
   n42 = "Transform (Modify) Bone - Bone: dyn_beltBack_01_l"(
    ComponentPose=Local To Component(LocalPose=Use cached pose 'SwimJump Fix Cache'())
    Alpha=@n41
    Translation="0,0,1"
    Rotation="0,0,0"
    Scale="1,1,1"
   ) {
    rotationMode=BMM_Additive
    rotationSpace=BCS_BoneSpace
    translationMode=BMM_Additive
    translationSpace=BCS_BoneSpace
    alphaBoolBlend {blendOption=Linear}
    alphaScaleBiasClamp {bInterpResult=true, interpSpeedIncreasing=15}
    boneToModify {boneName=dyn_beltBack_01_l}
    translation {z=1}
   }
   n43 = "Transform (Modify) Bone - Bone: dyn_beltBack_01_r"(ComponentPose=@n42, Alpha=@n41, Translation="0,0,1", Rotation="0,0,0", Scale="1,1,1") {
    rotationMode=BMM_Additive
    rotationSpace=BCS_BoneSpace
    translationMode=BMM_Additive
    translationSpace=BCS_BoneSpace
    alphaBoolBlend {blendOption=Linear}
    alphaScaleBiasClamp {bInterpResult=true, interpSpeedIncreasing=15}
    boneToModify {boneName=dyn_beltBack_01_r}
    translation {z=1}
   }
   n44 = Get PlayMelee() -> (PlayMelee:double)
   n45 = "Transform (Modify) Bone - Bone: dyn_beltBack_0_l"(ComponentPose=@n43, Alpha=@n44, Translation="0,0,3", Rotation="0,0,0", Scale="1,1,1") {
    rotationMode=BMM_Additive
    rotationSpace=BCS_BoneSpace
    translationMode=BMM_Additive
    translationSpace=BCS_BoneSpace
    alphaBoolBlend {blendOption=Linear}
    alphaScaleBiasClamp {bInterpResult=true, interpSpeedIncreasing=5}
    boneToModify {boneName=dyn_beltBack_0_l}
    translation {z=3}
   }
   n46 = "Transform (Modify) Bone - Bone: dyn_beltBack_02b_l"(ComponentPose=@n45, Alpha=@n44, Translation="0,-1,0.5", Rotation="0,0,0", Scale="1,1,1") {
    rotationMode=BMM_Additive
    rotationSpace=BCS_BoneSpace
    translationMode=BMM_Additive
    translationSpace=BCS_BoneSpace
    alphaBoolBlend {blendOption=Linear}
    alphaScaleBiasClamp {bInterpResult=true, interpSpeedDecreasing=5}
    boneToModify {boneName=dyn_beltBack_02b_l}
    translation {y=-1, z=0.5}
   }
   n47 = "Transform (Modify) Bone - Bone: base_beltBack_02_r"(ComponentPose=@n46, Alpha=Get sprinting() -> (sprinting:double), Translation="0.000,0.000,0.564", Rotation="0,0,0", Scale="1,1,1") {
    rotationMode=BMM_Additive
    rotationSpace=BCS_BoneSpace
    translationMode=BMM_Additive
    translationSpace=BCS_BoneSpace
    alphaBoolBlend {blendOption=Linear}
    alphaScaleBiasClamp {bInterpResult=true}
    boneToModify {boneName=base_beltBack_02_r}
    translation {z=0.564}
   }
   n48 = Get Left() -> (Left:double)
   n49 = "Transform (Modify) Bone - Bone: dyn_beltBack_0_r"(ComponentPose=@n47, Alpha=@n48, Translation="0,0,4", Rotation="20,0,0", Scale="1,1,1") {
    rotationMode=BMM_Additive
    rotationSpace=BCS_BoneSpace
    translationMode=BMM_Additive
    translationSpace=BCS_BoneSpace
    alphaBoolBlend {blendOption=Linear}
    alphaScaleBiasClamp {bInterpResult=true}
    boneToModify {boneName=dyn_beltBack_0_r}
    rotation {pitch=20}
    translation {z=4}
   }
   n50 = "Transform (Modify) Bone - Bone: dyn_beltBack_01_r"(ComponentPose=@n49, Alpha=@n48, Translation="0,0,1", Rotation="0,0,0", Scale="1,1,1") {
    rotationMode=BMM_Additive
    rotationSpace=BCS_BoneSpace
    translationMode=BMM_Additive
    translationSpace=BCS_BoneSpace
    alphaBoolBlend {blendOption=Linear}
    alphaScaleBiasClamp {bInterpResult=true}
    boneToModify {boneName=dyn_beltBack_01_r}
    translation {z=1}
   }
   n51 = "Transform (Modify) Bone - Bone: dyn_beltBack_01_r"(ComponentPose=@n50, Alpha=Get Right() -> (Right:double), Translation="0,4,1", Rotation="0,0,0", Scale="1,1,1") {
    rotationMode=BMM_Additive
    rotationSpace=BCS_BoneSpace
    translationMode=BMM_Additive
    translationSpace=BCS_BoneSpace
    alphaBoolBlend {blendOption=Linear}
    alphaScaleBiasClamp {bInterpResult=true, interpSpeedDecreasing=5, interpSpeedIncreasing=5}
    boneToModify {boneName=dyn_beltBack_01_r}
    translation {y=4, z=1}
   }
   n52 = Get Alpha_isDBNO() -> (Alpha_isDBNO:double)
   n53 = "Transform (Modify) Bone - Bone: dyn_beltBack_01_r"(ComponentPose=@n51, Alpha=@n52, Translation="0,4,0", Rotation="0,0,0", Scale="1,1,1") {
    rotationMode=BMM_Additive
    rotationSpace=BCS_BoneSpace
    translationMode=BMM_Additive
    translationSpace=BCS_BoneSpace
    alphaBoolBlend {blendOption=Linear}
    alphaScaleBiasClamp {bInterpResult=true, interpSpeedDecreasing=5, interpSpeedIncreasing=5}
    boneToModify {boneName=dyn_beltBack_01_r}
    translation {y=4}
   }
   n54 = "Transform (Modify) Bone - Bone: dyn_beltBack_0_r"(ComponentPose=@n53, Alpha=@n52, Translation="0,0,-3", Rotation="0,0,0", Scale="1,1,1") {
    rotationMode=BMM_Additive
    rotationSpace=BCS_BoneSpace
    translationMode=BMM_Additive
    translationSpace=BCS_BoneSpace
    alphaBoolBlend {blendOption=Linear}
    alphaScaleBiasClamp {bInterpResult=true, interpSpeedDecreasing=5, interpSpeedIncreasing=5}
    boneToModify {boneName=dyn_beltBack_0_r}
    translation {z=-3}
   }
   n55 = "Transform (Modify) Bone - Bone: dyn_beltBack_0_l"(ComponentPose=@n54, Alpha=@n52, Translation="0,0,-3", Rotation="0,0,0", Scale="1,1,1") {
    rotationMode=BMM_Additive
    rotationSpace=BCS_BoneSpace
    translationMode=BMM_Additive
    translationSpace=BCS_BoneSpace
    alphaBoolBlend {blendOption=Linear}
    alphaScaleBiasClamp {bInterpResult=true, interpSpeedDecreasing=5, interpSpeedIncreasing=5}
    boneToModify {boneName=dyn_beltBack_0_l}
    translation {z=-3}
   }
   Save cached pose 'ExtraPoses Fix Cache'(Pose=Component To Local(ComponentPose=@n55)) {cacheName=ExtraPoses Fix Cache}
   n56 = Get Alpha_SwimJump() -> (Alpha_SwimJump:double)
   n57 = "Transform (Modify) Bone - Bone: dyn_beltBack_01_l"(
    ComponentPose=Local To Component(LocalPose=Use cached pose 'SkyDiveFloating Fix Cache'())
    Alpha=@n56
    Translation="0,3,-1"
    Rotation="0,0,0"
    Scale="1,1,1"
   ) {
    rotationMode=BMM_Additive
    rotationSpace=BCS_BoneSpace
    translationMode=BMM_Additive
    translationSpace=BCS_BoneSpace
    alphaBoolBlend {blendOption=Linear}
    alphaScaleBiasClamp {bInterpResult=true, interpSpeedDecreasing=5, interpSpeedIncreasing=5}
    boneToModify {boneName=dyn_beltBack_01_l}
    translation {y=3, z=-1}
   }
   n58 = "Transform (Modify) Bone - Bone: dyn_beltBack_0_l"(ComponentPose=@n57, Alpha=@n56, Translation="0,0,-4", Rotation="0,0,0", Scale="1,1,1") {
    rotationMode=BMM_Additive
    rotationSpace=BCS_BoneSpace
    translationMode=BMM_Additive
    translationSpace=BCS_BoneSpace
    alphaBoolBlend {blendOption=Linear}
    alphaScaleBiasClamp {bInterpResult=true, interpSpeedDecreasing=5, interpSpeedIncreasing=5}
    boneToModify {boneName=dyn_beltBack_0_l}
    translation {z=-4}
   }
   n59 = "Transform (Modify) Bone - Bone: dyn_beltBack_01_r"(ComponentPose=@n58, Alpha=@n56, Translation="0,4,0", Rotation="0,0,0", Scale="1,1,1") {
    rotationMode=BMM_Additive
    rotationSpace=BCS_BoneSpace
    translationMode=BMM_Additive
    translationSpace=BCS_BoneSpace
    alphaBoolBlend {blendOption=Linear}
    alphaScaleBiasClamp {bInterpResult=true, interpSpeedDecreasing=5, interpSpeedIncreasing=5}
    boneToModify {boneName=dyn_beltBack_01_r}
    translation {y=4}
   }
   n60 = "Transform (Modify) Bone - Bone: dyn_beltBack_02_r"(ComponentPose=@n59, Alpha=@n56, Translation="0,1,0", Rotation="0,10,0", Scale="1,1,1") {
    rotationMode=BMM_Additive
    rotationSpace=BCS_BoneSpace
    translationMode=BMM_Additive
    translationSpace=BCS_BoneSpace
    alphaBoolBlend {blendOption=Linear}
    alphaScaleBiasClamp {bInterpResult=true, interpSpeedDecreasing=5, interpSpeedIncreasing=5}
    boneToModify {boneName=dyn_beltBack_02_r}
    rotation {yaw=10}
    translation {y=1}
   }
   n61 = Get Alpha_SwimFalling() -> (Alpha_SwimFalling:double)
   n62 = "Transform (Modify) Bone - Bone: dyn_beltBack_01_l"(ComponentPose=@n60, Alpha=@n61, Translation="0,5,3", Rotation="0,0,0", Scale="1,1,1") {
    rotationMode=BMM_Additive
    rotationSpace=BCS_BoneSpace
    translationMode=BMM_Additive
    translationSpace=BCS_BoneSpace
    alphaBoolBlend {blendOption=Linear}
    alphaScaleBiasClamp {bInterpResult=true, interpSpeedDecreasing=5, interpSpeedIncreasing=5}
    boneToModify {boneName=dyn_beltBack_01_l}
    translation {y=5, z=3}
   }
   n63 = "Transform (Modify) Bone - Bone: dyn_beltBack_0_l"(ComponentPose=@n62, Alpha=@n61, Translation="0,0,-15", Rotation="0,0,0", Scale="1,1,1") {
    rotationMode=BMM_Additive
    rotationSpace=BCS_BoneSpace
    translationMode=BMM_Additive
    translationSpace=BCS_BoneSpace
    alphaBoolBlend {blendOption=Linear}
    alphaScaleBiasClamp {bInterpResult=true, interpSpeedDecreasing=5, interpSpeedIncreasing=5}
    boneToModify {boneName=dyn_beltBack_0_l}
    translation {z=-15}
   }
   n64 = "Transform (Modify) Bone - Bone: dyn_beltBack_02b_l"(ComponentPose=@n63, Alpha=@n61, Translation="0,0,-3", Rotation="0,0,0", Scale="1,1,1") {
    rotationMode=BMM_Additive
    rotationSpace=BCS_BoneSpace
    translationMode=BMM_Additive
    translationSpace=BCS_BoneSpace
    alphaBoolBlend {blendOption=Linear}
    alphaScaleBiasClamp {bInterpResult=true, interpSpeedDecreasing=5, interpSpeedIncreasing=5}
    boneToModify {boneName=dyn_beltBack_02b_l}
    translation {z=-3}
   }
   n65 = "Transform (Modify) Bone - Bone: dyn_beltBack_01_r"(ComponentPose=@n64, Alpha=@n61, Translation="0,-2,6", Rotation="0,0,0", Scale="1,1,1") {
    rotationMode=BMM_Additive
    rotationSpace=BCS_BoneSpace
    translationMode=BMM_Additive
    translationSpace=BCS_BoneSpace
    alphaBoolBlend {blendOption=Linear}
    alphaScaleBiasClamp {bInterpResult=true, interpSpeedDecreasing=5, interpSpeedIncreasing=5}
    boneToModify {boneName=dyn_beltBack_01_r}
    translation {y=-2, z=6}
   }
   n66 = "Transform (Modify) Bone - Bone: dyn_beltBack_02_r"(ComponentPose=@n65, Alpha=@n61, Translation="0,0,-1", Rotation="0,20,0", Scale="1,1,1") {
    rotationMode=BMM_Additive
    rotationSpace=BCS_BoneSpace
    translationMode=BMM_Additive
    translationSpace=BCS_BoneSpace
    alphaBoolBlend {blendOption=Linear}
    alphaScaleBiasClamp {bInterpResult=true, interpSpeedDecreasing=5, interpSpeedIncreasing=5}
    boneToModify {boneName=dyn_beltBack_02_r}
    rotation {yaw=20}
    translation {z=-1}
   }
   n67 = "Transform (Modify) Bone - Bone: dyn_beltBack_0_r"(ComponentPose=@n66, Alpha=@n61, Translation="0,5,-12", Rotation="0,0,0", Scale="1,1,1") {
    rotationMode=BMM_Additive
    rotationSpace=BCS_BoneSpace
    translationMode=BMM_Additive
    translationSpace=BCS_BoneSpace
    alphaBoolBlend {blendOption=Linear}
    alphaScaleBiasClamp {bInterpResult=true, interpSpeedDecreasing=5, interpSpeedIncreasing=5}
    boneToModify {boneName=dyn_beltBack_0_r}
    translation {y=5, z=-12}
   }
   Save cached pose 'SwimJump Fix Cache'(Pose=Component To Local(ComponentPose=@n67)) {cacheName=SwimJump Fix Cache}
  ]}
 EventGraph {nodes [
   Event BlueprintInitializeAnimation()
   > Set SkelMesh(SkelMesh=GetOwnerSkeletalMeshComponent())
   n1 = TryGetPawnOwner()
   > Is Valid(InputObject=@n1)
   > n2 = Cast To FortPlayerPawnAthena(Object=@n1)
   > n3 = Set PlayerPawn(PlayerPawn:Fort Player Pawn Athena Object Reference=@n2) -> (Output_Get:Fort Player Pawn Athena Object Reference)
   > Set Body_Mesh(Body_Mesh=GetSkeletalMeshForPartType(self=@n3))
   n4 = Event BlueprintUpdateAnimation()
   > n5 = Sequence!(execute) -> (
    then_0
    then_1
    then_2:Exec
    then_3:Exec
    then_4:Exec
    then_5:Exec
    then_6:Exec
    then_7:Exec
    then_8:Exec
    then_9:Exec
    then_10:Exec
    then_11:Exec
    then_12:Exec
    then_13:Exec
    then_14:Exec
   )
   > Sequence!(execute) -> (then_0, then_1)
   > Is Valid(InputObject=Get PlayerPawn() -> (PlayerPawn:Fort Player Pawn Athena Object Reference))
   > n6 = Branch(Condition=Get bIsInFrontEnd())
   Set WPO Globals Function(
    execute=@n6.else
    Material=Get Mat Instance Dynamic()
    Overlay Material=Get Overlay Mat Instance Dynamic()
    Skeletal Mesh=Get SkelMesh()
    "Local Noise Transform  Location"=Get Local Noise Transform Location() -> (Local Noise Transform Location:Vector)
    Local Noise Offset=Get Local Noise Offset()
   )
   > Set WPO Regional Function!(
    execute
    Material:Material Instance Dynamic Object Reference=Get Mat Instance Dynamic()
    Overlay Material:Material Instance Dynamic Object Reference=Get Overlay Mat Instance Dynamic()
    Skeletal Mesh:Skeletal Mesh Component Object Reference=Get SkelMesh()
    Velocity Transform:Name=lowerarm_l
    Velocity Interp Speed:double=Get Velocity Interp Speed()
    Collision Transform:Name=upperarm_l
    Collision Transform Axis:EAxis Enum
    Invert Collision Transform Axis:bool=false
    Delta Time X:double=@n4.DeltaTimeX
    Previous Velocity Transform Location:"Vector (by ref)"=Get R1 Velocity Transform Location()
    Previous Velocity Transform Velocity:"Vector (by ref)"=Get R1 Velocity Transform Velocity()
    Region:int=1
   ) -> (then)
   > Set WPO Regional Function(
    Material=Get Mat Instance Dynamic()
    Overlay Material=Get Overlay Mat Instance Dynamic()
    Skeletal Mesh=Get SkelMesh()
    Velocity Transform=lowerarm_r
    Velocity Interp Speed=Get Velocity Interp Speed()
    Collision Transform=upperarm_r
    Invert Collision Transform Axis=true
    Delta Time X=@n4.DeltaTimeX
    Previous Velocity Transform Location=Get R2 Velocity Transform Location()
    Previous Velocity Transform Velocity=Get R2 Velocity Transform Velocity()
    Region=2
   )
   > n7 = Branch(Condition=OR Boolean(A=Get bIsSkydiveFloating(), B=Get bIsParachuteOpened()))
   > Set WPO Pho Wind Intensity(WPO Pho Wind Intensity=0.7)
   > n8 = Set WPO Pho Turbulence(WPO Pho Turbulence:double=0.5) -> (Output_Get:double)
   SetScalarParameterValue(execute=[@n10.then, @n8.then], self=Get Mat Instance Dynamic(), ParameterName=ClothSim_WindStrength, Value=Get WPO Pho Wind Intensity())
   n9 = Get WPO Pho Turbulence() -> (WPO Pho Turbulence:double)
   > SetScalarParameterValue(self=Get Mat Instance Dynamic(), ParameterName=ClothSim_R1_TurbulenceStrength, Value=@n9)
   > SetScalarParameterValue(self=Get Mat Instance Dynamic(), ParameterName=ClothSim_R2_TurbulenceStrength, Value=@n9)
   > SetScalarParameterValue(self=Get Mat Instance Dynamic(), ParameterName=ClothSim_R3_TurbulenceStrength, Value=@n9)
   Set WPO Pho Wind Intensity(execute=@n7.else, WPO Pho Wind Intensity=0)
   > n10 = Set WPO Pho Turbulence(WPO Pho Turbulence=0)
   Set PlayMelee(execute=@n5.then_2, PlayMelee:double="To Float (Boolean)"(InBool=Get bIsPlayingMeleeAnim())) -> (Output_Get:double)
   > Set Is Sprinting Sky Dive(Is Sprinting Sky Dive:bool=OR Boolean!(A=Get bIsSprinting(), B=Get bIsSkydiveFloating(), C:bool=Get bIsSkydiveDiveMode(), D:bool=Get bIsParachuteOpened()) -> (ReturnValue)) -> (Output_Get:bool)
   > Set sprinting(sprinting:double="To Float (Boolean)"(InBool=Get bIsSprinting())) -> (Output_Get:double)
   n11 = Branch(execute=@n5.then_3, Condition=Get bIsSkydiveFloating())
   > Set Alpha_SkyDiveFloating(Alpha_SkyDiveFloating:float=1) -> (Output_Get:float)
   Set Alpha_SkyDiveFloating(execute=@n11.else, Alpha_SkyDiveFloating=0)
   n12 = Branch(execute=@n5.then_4, Condition=Get bIsParachuteOpened())
   > Set Alpha_openParachused(Alpha_openParachused:float=1) -> (Output_Get:float)
   Set Alpha_openParachused(execute=@n12.else, Alpha_openParachused=0)
   n13 = Gravity Override State Function DEPRECATED(
    execute=@n5.then_5
    CharacterPartAnimInstance=Self-Reference()
    GameTimeInSeconds=GetGameTimeInSeconds()
    GravityOverrideParams=Get Gravity Override Params()
   )
   > Set GravityOverride(GravityOverride=@n13)
   n14 = Branch(execute=@n5.then_6, Condition=Get bIsTargeting())
   > Set Alpha_isTargeting(Alpha_isTargeting:double=1) -> (Output_Get:double)
   Set Alpha_isTargeting(execute=@n14.else, Alpha_isTargeting=0)
   n15 = Get bIsSprinting()
   n16 = Get bIsTacticalSprinting()
   n17 = Branch(execute=@n5.then_11, Condition=OR Boolean!(A=NOT Boolean(A=@n15), B=NOT Boolean(A=@n16)) -> (ReturnValue))
   > Set Alpha_NoSprinting(Alpha_NoSprinting:double=1) -> (Output_Get:double)
   Set Alpha_NoSprinting(execute=@n17.else, Alpha_NoSprinting=0)
   n18 = OR Boolean(A=Get bIsJumping(), B=Get bIsFalling())
   n19 = OR Boolean(A=Get bIsRelaxedLevel2() -> (bIsRelaxedLevel2:bool), B=Get bIsRelaxedLevel1() -> (bIsRelaxedLevel1:bool))
   n20 = AND Boolean(A=NOT Boolean(A=Get bIsCrouching()), B=Get bIsSprinting())
   n21 = Branch(execute=@n5.then_1, Condition=AND Boolean(A=AND Boolean(A=@n20, B=@n19), B=NOT Boolean(A=@n18)))
   > Set Alpha_isSprintingNoCrouch(Alpha_isSprintingNoCrouch:double=1) -> (Output_Get:double)
   Set Alpha_isSprintingNoCrouch(execute=@n21.else, Alpha_isSprintingNoCrouch=0)
   n22 = Branch(execute=@n5.then_12, Condition=AND Boolean(A=Get bIsWaterJump() -> (bIsWaterJump:bool), B=Get bIsSprinting()))
   > Set Alpha_SwimJump(Alpha_SwimJump:double=1) -> (Output_Get:double)
   Set Alpha_SwimJump(execute=@n22.else, Alpha_SwimJump=0)
   n23 = Branch(execute=@n5.then_13, Condition=AND Boolean!(A=Get bIsWaterJump(), B=Get bIsSprinting(), C:bool=Get bIsFalling()) -> (ReturnValue))
   > Set Alpha_SwimFalling(Alpha_SwimFalling:double=1) -> (Output_Get:double)
   Set Alpha_SwimFalling(execute=@n23.else, Alpha_SwimFalling=0)
   n24 = Get bIsMoving2D()
   n25 = Branch(execute=@n5.then_14, Condition=AND Boolean!(A=Get bIsDBNO(), B=NOT Boolean(A=@n24)) -> (ReturnValue))
   > Set Alpha_isDBNO(Alpha_isDBNO:double=1) -> (Output_Get:double)
   Set Alpha_isDBNO(execute=@n25.else, Alpha_isDBNO=0)
   n26 = AND Boolean(A=OR Boolean(A=Get bIsJumping(), B=Get bIsFalling()), B=NOT Boolean(A=Get bIsSkyDiving()))
   n27 = Branch(execute=@n5.then_8, Condition=@n26)
   > Set Alpha_jumping(Alpha_jumping:float=1) -> (Output_Get:float)
   Set Alpha_jumping(execute=@n27.else, Alpha_jumping=0)
   n28 = Branch(execute=@n5.then_7, Condition=Get bIsSkydiveDiveMode())
   > Set Alpha_skydiveMode(Alpha_skydiveMode:float=1) -> (Output_Get:float)
   Set Alpha_skydiveMode(execute=@n28.else, Alpha_skydiveMode=0)
   n29 = PlayerMovement(execute=@n5.then_9, AnimInstance=Self-Reference(), XDotProductCoefficient=100, YDotProductCoefficient=100, ZDotProductCoefficient=100)
   n30 = BreakVector(InVec=MakeVector(X=@n29, Y=@n29.YMovement))
   > Set Left(Left:double="To Float (Boolean)"(InBool="float < float"(A=@n30, B=-50))) -> (Output_Get:double)
   n31 = NOT Boolean(A=OR Boolean(A=Get bIsJumping(), B=Get bIsFalling()))
   n32 = "float > float"(A=@n30, B=50)
   > Set Right(Right:double="To Float (Boolean)"(InBool=AND Boolean(A=@n32, B=@n31))) -> (Output_Get:double)
   > Set Forward(Forward:double="To Float (Boolean)"(InBool="float > float"(A=@n30.Y, B=40))) -> (Output_Get:double)
   > Set Backward(Backward:double="To Float (Boolean)"(InBool="float < float"(A=@n30.Y, B=-40))) -> (Output_Get:double)
   n33 = Branch(execute=@n5.then_10, Condition=Get bIsSprinting())
   > Set Alpha_Sprinting(Alpha_Sprinting:double=1) -> (Output_Get:double)
   Set Alpha_Sprinting(execute=@n33.else, Alpha_Sprinting=0)
   Event OnCharacterCustomizationLoaded()
   > n34 = Set SkelMesh(SkelMesh=GetPartSkeletalMeshComponent())
   > n35 = Get Dynamic Material Instance(SkeletalMeshComponent=@n34)
   > Set Mat Instance Dynamic(Mat Instance Dynamic=@n35)
   > n36 = GetOrCreateDynamicOverlayMaterial(MeshComponent=@n34, MaterialName=None)
   > Set Overlay Mat Instance Dynamic(Overlay Mat Instance Dynamic=@n36)
   AND Boolean(A=@n26, B=NOT Boolean(A=Get bIsSprinting()))
   Get bIsSprinting()
  ]}
]
variables [
 Alpha_isDBNO {type=number}
 Alpha_isSprintingNoCrouch {type=number}
 Alpha_isTargeting {type=number}
 Alpha_jumping {type=number, default=1}
 Alpha_NoSprinting {type=number}
 Alpha_openParachused {type=number}
 Alpha_SkyDiveFloating {type=number}
 Alpha_skydiveMode {type=number, default=1}
 Alpha_Sprinting {type=number}
 Alpha_SwimFalling {type=number}
 Alpha_SwimJump {type=number}
 Backward {type=number}
 Body_Mesh {type=SkeletalMeshComponent}
 Forward {type=number}
 Gravity Override Params {type=GravityOverrideParamsStruct}
 GravityOverride {type=Vector}
 Is Sprinting Sky Dive {type=boolean}
 Left {type=number}
 Local Noise Offset {type=Vector}
 Local Noise Transform Location {type=Vector}
 Mat Instance Dynamic {type=MaterialInstanceDynamic}
 Overlay Mat Instance Dynamic {type=MaterialInstanceDynamic}
 PlayerPawn {type=FortPlayerPawnAthena}
 PlayMelee {type=number}
 R1 Velocity Transform Location {type=Vector}
 R1 Velocity Transform Velocity {type=Vector}
 R2 Velocity Transform Location {type=Vector}
 R2 Velocity Transform Velocity {type=Vector}
 Right {type=number}
 SkelMesh {type=SkeletalMeshComponent}
 sprinting {type=number}
 Velocity Interp Speed {type=number}
 WPO Pho Turbulence {type=number}
 WPO Pho Wind Intensity {type=number}
]
