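# Control Rig for the GroovyReader female face accessory part that applies pose-driven corrective offsets to its hair bones.
# A sequence runs spherical pose readers on the right upper arm and the head, whose outputs weight Modify Transforms on hair bones.
# An extra right side-hair offset scales the arm weight by isCrouching.
# A second, opposite-facing head reader chain is not connected to execution, and getTargetingState is unused.
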ControlRigBlueprint /BRCosmetics/Characters/Player/Female/Medium/Bodies/F_MED_GroovyReader/Meshes/Parts/F_MED_GroovyReader_FaceAcc_CtrlRig
graphs [
 RigVMFunctionLibrary {links [], nodes []}
 RigVMModel {nodes [
   RigUnit_BeginExecution = &RigUnit_BeginExecution()
   SequenceAggregate = &RigVMFunction_Sequence(ExecuteContext=@RigUnit_BeginExecution.ExecutePin)
   n1 = &RigUnit_SphericalPoseReader(
    ActiveRegionScaleFactors=(PositiveWidth=1,NegativeWidth=1,PositiveHeight=1,NegativeHeight=1)
    ActiveRegionSize=0.1
    Debug=(bDrawDebug=True,bDraw2D=False,bDrawLocalAxes=False,DebugScale=25,DebugSegments=20,DebugThickness=0.25)
    DriverAxis=(X=1,Y=0,Z=0)
    DriverItem=(Type=Bone,Name="upperarm_r")
    FalloffRegionScaleFactors=(PositiveWidth=1,NegativeWidth=1,PositiveHeight=1,NegativeHeight=1)
    FalloffSize=0.4
    FlipHeightScaling=False
    FlipWidthScaling=False
    OptionalParentItem=(Type=Bone,Name="None")
    RotationOffset=(X=90,Y=0,Z=10)
    ExecutePin=@SequenceAggregate.A
   )
   > &RigUnit_ModifyTransforms(
    ItemToModify="((Item=(Type=Bone,Name=\"base_hairBack\"),Transform=(Rotation=(X=0,Y=0,Z=0,W=1),Translation=(X=0,Y=0,Z=1),Scale3D=(X=1,Y=1,Z=1))),(Item=(Type=Bone,Name=\"base_hairSide1_r\"),Transform=(Rotation=(X=-0.083,Y=-0.0509,Z=0.083,W=0.9918),Translation=(X=0,Y=0.7,Z=0), ...(truncated)"
    Weight=@n1.OutputParam
   )
   > &RigUnit_ModifyTransforms(
    ItemToModify=((Item=(Type=Bone,Name="base_hairSide1_r"),Transform=(Rotation=(X=0,Y=0,Z=0.0698,W=0.9976),Translation=(X=0,Y=2,Z=0),Scale3D=(X=1,Y=1,Z=1))))
    Weight=&RigVMFunction_MathDoubleMul(A=@n1.OutputParam, B=RigVMVariableNode() {variable=isCrouching}.Value).Result
   )
   n2 = &RigUnit_SphericalPoseReader(
    ActiveRegionScaleFactors=(PositiveWidth=1,NegativeWidth=1,PositiveHeight=1,NegativeHeight=1)
    ActiveRegionSize=0.1
    Debug=(bDrawDebug=true,bDraw2D=False,bDrawLocalAxes=False,DebugScale=25,DebugSegments=20,DebugThickness=0.25)
    DriverAxis=(X=1,Y=0,Z=0)
    DriverItem=(Type=Bone,Name="head")
    FalloffRegionScaleFactors=(PositiveWidth=1,NegativeWidth=1,PositiveHeight=1,NegativeHeight=1)
    FalloffSize=0.18
    FlipHeightScaling=false
    FlipWidthScaling=false
    OptionalParentItem=(Type=Bone)
    RotationOffset=(X=90,Y=0,Z=-125)
    ExecutePin=@SequenceAggregate.B
   )
   > &RigUnit_ModifyTransforms(
    ItemToModify="((Item=(Type=Bone,Name=\"base_hairBack\"),Transform=(Rotation=(X=0,Y=-0.2588,Z=0,W=0.9659),Translation=(X=0,Y=0,Z=0),Scale3D=(X=1,Y=1,Z=1))),(Item=(Type=Bone,Name=\"base_hairSide_c\"),Transform=(Rotation=(X=0,Y=0,Z=0.342,W=0.9397),Translation=(X=0,Y=0,Z=0),Sca ...(truncated)"
    WeightMinimum=0
    Weight=@n2.OutputParam
   )
   n3 = &RigUnit_SphericalPoseReader(
    ActiveRegionScaleFactors=(PositiveWidth=1,NegativeWidth=1,PositiveHeight=1,NegativeHeight=1)
    ActiveRegionSize=0.05
    Debug=(bDrawDebug=true,bDraw2D=False,bDrawLocalAxes=False,DebugScale=25,DebugSegments=20,DebugThickness=0.25)
    DriverAxis=(X=1,Y=0,Z=0)
    DriverItem=(Type=Bone,Name="head")
    ExecutePin=()
    FalloffRegionScaleFactors=(PositiveWidth=1,NegativeWidth=1,PositiveHeight=1,NegativeHeight=1)
    FalloffSize=0.18
    FlipHeightScaling=false
    FlipWidthScaling=false
    OptionalParentItem=(Type=Bone)
    RotationOffset=(X=-90,Y=0,Z=125)
   )
   > &RigUnit_ModifyTransforms(
    ItemToModify="((Item=(Type=Bone,Name=\"dyn_hairBack\"),Transform=(Rotation=(X=0,Y=-0.0872,Z=0,W=0.9962),Translation=(X=0,Y=0,Z=0),Scale3D=(X=1,Y=1,Z=1))),(Item=(Type=Bone,Name=\"dyn_hairBack_d\"),Transform=(Rotation=(X=0,Y=-0.0436,Z=0,W=0.999),Translation=(X=0,Y=0,Z=0),Scal ...(truncated)"
    WeightMinimum=0
    Weight=@n3.OutputParam
   )
   > &RigUnit_ModifyTransforms(
    ItemToModify="((Item=(Type=Bone,Name=\"base_hairBack\"),Transform=(Rotation=(X=0,Y=-0.2588,Z=0,W=0.9659),Translation=(X=0,Y=0,Z=0),Scale3D=(X=1,Y=1,Z=1))),(Item=(Type=Bone,Name=\"None\"),Transform=(Rotation=(X=0,Y=0,Z=0,W=1),Translation=(X=0,Y=0,Z=0),Scale3D=(X=1,Y=1,Z=1))) ...(truncated)"
    WeightMinimum=0
    Weight=&RigVMFunction_MathDoubleMul(B=1, A=@n3.OutputParam).Result
   )
  ]}
]
hierarchy [
 root {loc=<0 0 0>, rot=<0 0 0 1>}
 pelvis {parent=root, loc=<0 -0.5301 98.0571>, rot=<0.0127 0.707 -0.0127 -0.707>}
 spine_01 {parent=pelvis, loc=<6.1208 -0.9328 0>, rot=<0 0 0.034 -0.9994>}
 spine_02 {parent=spine_01, loc=<6.5996 -0.2095 0>, rot=<0 0 0.0238 -0.9997>}
 spine_03 {parent=spine_02, loc=<6.4914 -0.1237 0>, rot=<0 0 0.0149 -0.9999>}
 spine_04 {parent=spine_03, loc=<6.9598 0.0794 0>, rot=<0 0 -0.0132 -0.9999>}
 spine_05 {parent=spine_04, loc=<6.9987 0 0>, rot=<0 0 -0.0015 -1>}
 neck_01 {parent=spine_05, loc=<11.3138 0.1529 0>, rot=<0 0 -0.2003 -0.9797>}
 neck_02 {parent=neck_01, loc=<4.5627 0 0>, rot=<0 0 0.0504 -0.9987>}
 head {parent=neck_02, loc=<5.3914 0 0>, rot=<0 0 0.115 -0.9934>}
 base_hairTop {parent=head, loc=<15.71 -8.8213 0.0228>, rot=<0.0879 -0.7016 0.7016 -0.0879>}
 dyn_hairTop_01 {parent=base_hairTop, loc=<0 0 0>, rot=<0 0 0 1>}
 dyn_hairTop_02 {parent=dyn_hairTop_01, loc=<7.2828 0 0.6639>, rot=<-0.702 -0.085 0.085 -0.702>}
 dyn_hairTop_a {parent=dyn_hairTop_02, loc=<2.0505 -1.9525 -4.3968>, rot=<0.8914 0.0885 -0.3733 -0.2413>}
 dyn_hairTop_a01 {parent=dyn_hairTop_a, loc=<2.795 0 0>, rot=<0 0 0 1>}
 dyn_hairTop_b {parent=dyn_hairTop_02, loc=<3.8883 -0.7733 -3.3249>, rot=<0.9013 0.0702 -0.0826 -0.4195>}
 dyn_hairTop_b01 {parent=dyn_hairTop_b, loc=<2.795 0 0>, rot=<0 0 0 1>}
 dyn_hairTop_c {parent=dyn_hairTop_02, loc=<3.9131 0.9247 -0.8702>, rot=<0 0 0 1>}
 dyn_hairTop_c01 {parent=dyn_hairTop_c, loc=<3.8942 0 0>, rot=<0 0 0 1>}
 dyn_hairTop_d {parent=dyn_hairTop_02, loc=<4.0781 0.5874 2.2991>, rot=<0 0 0 1>}
 dyn_hairTop_d01 {parent=dyn_hairTop_d, loc=<3.9023 0 0>, rot=<0 0 0 1>}
 dyn_hairTop_e {parent=dyn_hairTop_02, loc=<2.7769 -1.6294 4.5828>, rot=<0.4023 0.0877 0.0087 -0.9113>}
 dyn_hairTop_e01 {parent=dyn_hairTop_e, loc=<3.8978 0 -0.1572>, rot=<0 0 0 1>}
 dyn_hairTop_f {parent=dyn_hairTop_02, loc=<2.189 -2.4755 5.5467>, rot=<0.3273 0.2059 0.1674 -0.9069>}
 dyn_hairTop_f01 {parent=dyn_hairTop_f, loc=<2.7442 0 0>, rot=<0.7237 0.0802 -0.253 -0.6371>}
 base_hairBack {parent=head, loc=<3.8564 -6.8767 0>, rot=<0 -0.7071 0.7071 0>}
 dyn_hairBack {parent=base_hairBack, loc=<0 0 0>, rot=<0 0 0 1>}
 dyn_hairBack_a {parent=dyn_hairBack, loc=<4.5631 4.834 0.1872>, rot=<0 0.0041 0 -1>}
 dyn_hairBack_a1 {parent=dyn_hairBack_a, loc=<3.5838 0 0>, rot=<0 0 0 1>}
 dyn_hairBack_b {parent=dyn_hairBack, loc=<4.4935 2.4034 2.0809>, rot=<0 0.0041 0 -1>}
 dyn_hairBack_b1 {parent=dyn_hairBack_b, loc=<3.8586 0 0>, rot=<0 0 0 1>}
 dyn_hairBack_c {parent=dyn_hairBack, loc=<4.4936 -2.4034 2.0809>, rot=<0 0.0041 0 -1>}
 dyn_hairBack_c1 {parent=dyn_hairBack_c, loc=<3.859 0 0>, rot=<0 0 0 1>}
 dyn_hairBack_d {parent=dyn_hairBack, loc=<4.5632 -4.834 0.1872>, rot=<0 0.0041 0 -1>}
 dyn_hairBack_d1 {parent=dyn_hairBack_d, loc=<3.584 0 0>, rot=<0 0 0 1>}
 base_hairBack_extra {parent=dyn_hairBack, loc=<0 0 0>, rot=<-0.0142 0.2035 -0.0166 -0.9788>}
 dyn_hairBack_extra {parent=base_hairBack_extra, loc=<0 0 0>, rot=<0 0 0 1>}
 dyn_hairBack_extra_l {parent=dyn_hairBack_extra, loc=<0.0294 2.7794 0.0036>, rot=<0 0 0 1>}
 dyn_hairBack_extra_r {parent=dyn_hairBack_extra, loc=<-0.1785 -2.7675 0.188>, rot=<-0.0268 -0.999 0.0345 0>}
 dyn_hairBack_a4 {parent=dyn_hairBack, loc=<4.5668 -4.834 -0.2427>, rot=<-0.1226 -0.8661 -0.4464 -0.1888>}
 base_hairSide_c {parent=head, loc=<5.2899 -4.2677 0>, rot=<0 0 0 1>}
 base_hairSide1_l {parent=base_hairSide_c, loc=<0.2977 0.5261 -6.1898>, rot=<0.0041 -1 0 0>}
 dyn_hairSide1_l {parent=base_hairSide1_l, loc=<0 0 0>, rot=<0 0 0 1>}
 dyn_hairSide2_l {parent=dyn_hairSide1_l, loc=<3.795 0 0>, rot=<0 0 0 1>}
 dyn_hairSide_l_a {parent=dyn_hairSide2_l, loc=<2.6119 0 0>, rot=<0.0067 0.2888 0.0223 -0.9571>}
 dyn_hairSide_l_a01 {parent=dyn_hairSide_l_a, loc=<4.6518 0 0>, rot=<-0.0067 -0.2888 -0.0223 -0.9571>}
 dyn_hairSide_l_b {parent=dyn_hairSide2_l, loc=<3.6692 1.597 -1.0995>, rot=<0 0 0 1>}
 dyn_hairSide_l_b01 {parent=dyn_hairSide_l_b, loc=<2.5382 0 0>, rot=<0 0 0 1>}
 base_hairSide1_r {parent=base_hairSide_c, loc=<0.2132 0.8094 6.2127>, rot=<0 0 1 0>}
 dyn_hairSide1_r {parent=base_hairSide1_r, loc=<0 0 0>, rot=<0 0 -0.0041 -1>}
 dyn_hairSide2_r {parent=dyn_hairSide1_r, loc=<3.6169 0 0>, rot=<0 0 0 1>}
 dyn_hairSide_r_a {parent=dyn_hairSide2_r, loc=<1.6475 -1.532 -0.6386>, rot=<0 0 0 1>}
 dyn_hairSide_r_a01 {parent=dyn_hairSide_r_a, loc=<3.899 0 0>, rot=<0 0 0 1>}
 dyn_hairSide_r_b {parent=dyn_hairSide2_r, loc=<0.9627 0.2927 1.6152>, rot=<-0.0599 0.1884 -0.0242 -0.98>}
 dyn_hairSide_r_b01 {parent=dyn_hairSide_r_b, loc=<3.98 0 0>, rot=<0 0 0 1>}
 dyn_hairSide_r_c {parent=dyn_hairSide2_r, loc=<2.4408 2.3521 1.9204>, rot=<0.0806 0.4418 -0.1448 -0.8817>}
 dyn_hairSide_r_c01 {parent=dyn_hairSide_r_c, loc=<2.3896 0.0276 0>, rot=<0 0 0 1>}
 base_hairBack_e {parent=base_hairSide_c, loc=<-2.8302 -3.0971 -6.0638>, rot=<0.1005 -0.8769 0.2903 -0.3696>}
 dyn_hairBack_e {parent=base_hairBack_e, loc=<0 0 0>, rot=<0 0 0 1>}
 dyn_hairBack_l_e1 {parent=dyn_hairBack_e, loc=<3.98 0 0>, rot=<0 0 0 1>}
 clavicle_l {parent=spine_05, loc=<5.4831 -0.1053 -3.2946>, rot=<0.0465 -0.7289 -0.009 -0.6829>}
 upperarm_l {parent=clavicle_l, loc=<9.8658 -1.2136 -0.4105>, rot=<0.0164 -0.3245 -0.0269 -0.9454>}
 clavicle_r {parent=spine_05, loc=<5.4831 -0.1053 3.2946>, rot=<0.7289 0.0465 -0.6829 0.009>}
 upperarm_r {parent=clavicle_r, loc=<-9.8658 1.2136 0.4105>, rot=<0.0164 -0.3245 -0.0269 -0.9454>}
]
variables [
 getTargetingState {type=float, default=1}
 isCrouching {type=double, default=0}
]
