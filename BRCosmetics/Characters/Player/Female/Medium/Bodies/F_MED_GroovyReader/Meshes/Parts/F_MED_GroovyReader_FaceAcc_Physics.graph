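# Physics asset for the GroovyReader female medium body's FaceAcc mesh part, defining secondary motion for its hair.
# Capsule bodies form branching dyn_hair chains for the top, back and left and right sides, constrained back to base_hair anchor bodies.
# Constraints use mostly narrow swing limits, looser at some roots and tips, with twist limits set to zero.
# An unconstrained right-clavicle capsule is also present, likely as a shoulder collider for the hair.
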PhysicsAsset /BRCosmetics/Characters/Player/Female/Medium/Bodies/F_MED_GroovyReader/Meshes/Parts/F_MED_GroovyReader_FaceAcc_Physics
bodies {
 base_hairBack [{shapeType=Capsule, length=4.0205, radius=3.9795, shapeName=base_hairBack_capsule, center=<11.9735 0.3915 -2.4216>, rotation=<rot 90 180 -90>}]
 base_hairBack_e [{shapeType=Capsule, length=0.505, radius=0.505, shapeName=base_hairBack_e_capsule, rotation=<rot 0 -90 0>}]
 base_hairSide1_l [{shapeType=Capsule, length=0.505, radius=0.505, shapeName=base_hairSide1_l_capsule, center=<1.2 0 0>, rotation=<rot 90 0 0>}]
 base_hairSide1_r [{shapeType=Capsule, length=0.505, radius=0.505, shapeName=base_hairSide1_r_capsule, center=<1.2 0 0>, rotation=<rot 90 0 0>}]
 base_hairTop [{
   shapeType=Capsule
   length=1.5376
   radius=6.4674
   shapeName=base_hairTop_capsule
   center=<12.7144 0 -8.0537>
   rotation=<rot 90 165.9638 -104.0362>
  }]
 clavicle_r [{
   shapeType=Capsule
   length=12.1498
   radius=6.6178
   shapeName=clavicle_r_capsule
   center=<-2.4868 0.4721 0.4225>
   rotation=<rot -0.0009 -89.9999 105.3542>
  }]
 dyn_hairBack [{shapeType=Capsule, length=2, radius=1, shapeName=dyn_hairBack_a_capsule, center=<1.2 0 0>, rotation=<rot 90 0 0>}]
 dyn_hairBack_a [{shapeType=Capsule, length=2, radius=1, shapeName=dyn_hairBack_a_capsule, center=<1.2 0 0>, rotation=<rot 90 0 0>}]
 dyn_hairBack_a1 [{shapeType=Capsule, length=2, radius=1.2, shapeName=dyn_hairBack_a_capsule, center=<1.2 0 0>, rotation=<rot 90 0 0>}]
 dyn_hairBack_b [{shapeType=Capsule, length=2, radius=1, shapeName=dyn_hairBack_a_capsule, center=<1.2 0 0>, rotation=<rot 90 0 0>}]
 dyn_hairBack_b1 [{shapeType=Capsule, length=2, radius=1, shapeName=dyn_hairBack_a_capsule, center=<1.2 0 0>, rotation=<rot 90 0 0>}]
 dyn_hairBack_c [{shapeType=Capsule, length=2, radius=1, shapeName=dyn_hairBack_a_capsule, center=<1.2 0 0>, rotation=<rot 90 0 0>}]
 dyn_hairBack_c1 [{shapeType=Capsule, length=2.6519, radius=1.809, shapeName=dyn_hairBack_a_capsule, center=<2.4923 0 0.6803>, rotation=<rot 84.004 -180 -180>}]
 dyn_hairBack_d [{shapeType=Capsule, length=2, radius=1, shapeName=dyn_hairBack_a_capsule, center=<1.2 0 0>, rotation=<rot 90 0 0>}]
 dyn_hairBack_d1 [{shapeType=Capsule, length=2.6519, radius=1.809, shapeName=dyn_hairBack_a_capsule, center=<1.7496 0 0>, rotation=<rot 90 0 0>}]
 dyn_hairBack_e [{shapeType=Capsule, length=2, radius=1, shapeName=dyn_hairBack_a_capsule, center=<1.2 0 0>, rotation=<rot 90 0 0>}]
 dyn_hairBack_l_e1 [{shapeType=Capsule, length=2, radius=1, shapeName=dyn_hairBack_a_capsule, center=<1.2 0 0>, rotation=<rot 90 0 0>}]
 dyn_hairSide1_l [{shapeType=Capsule, length=2, radius=1, shapeName=dyn_hairBack_a_capsule, center=<1.2 0 0>, rotation=<rot 90 0 0>}]
 dyn_hairSide1_r [{shapeType=Capsule, length=2, radius=1, shapeName=dyn_hairBack_a_capsule, center=<1.2 0 0>, rotation=<rot 90 0 0>}]
 dyn_hairSide2_l [{shapeType=Capsule, length=2, radius=1, shapeName=dyn_hairBack_a_capsule, center=<1.2 0 0>, rotation=<rot 90 0 0>}]
 dyn_hairSide2_r [{shapeType=Capsule, length=2, radius=1, shapeName=dyn_hairBack_a_capsule, center=<1.2 0 0>, rotation=<rot 90 0 0>}]
 dyn_hairSide_l_a [{shapeType=Capsule, length=1.5, radius=1, shapeName=dyn_hairSide_l_a_capsule, center=<1.2 0 0>, rotation=<rot 90 0 0>}]
 dyn_hairSide_l_a01 [{shapeType=Capsule, length=2, radius=1, shapeName=dyn_hairBack_a_capsule, center=<1.2 0 0>, rotation=<rot 90 0 0>}]
 dyn_hairSide_l_b [{shapeType=Capsule, length=2, radius=1, shapeName=dyn_hairBack_a_capsule, center=<1.2 0 0>, rotation=<rot 90 0 0>}]
 dyn_hairSide_l_b01 [{shapeType=Capsule, length=2, radius=1, shapeName=dyn_hairBack_a_capsule, center=<1.2 0 0>, rotation=<rot 90 0 0>}]
 dyn_hairSide_r_a [{shapeType=Capsule, length=2, radius=1, shapeName=dyn_hairBack_a_capsule, center=<1.2 0 0>, rotation=<rot 90 0 0>}]
 dyn_hairSide_r_a01 [{
   shapeType=Capsule
   length=2.1098
   radius=1.1866
   shapeName=dyn_hairBack_a_capsule
   center=<1.9236 0.2299 1.3856>
   rotation=<rot 90 180 176.5639>
  }]
 dyn_hairSide_r_b [{shapeType=Capsule, length=2, radius=1, shapeName=dyn_hairBack_a_capsule, center=<1.2 0 0>, rotation=<rot 90 0 0>}]
 dyn_hairSide_r_b01 [{
   shapeType=Capsule
   length=3.2061
   radius=1.2
   shapeName=dyn_hairBack_a_capsule
   center=<2.0305 -1.0175 -0.1477>
   rotation=<rot 76.1924 -10.0002 0.0001>
  }]
 dyn_hairSide_r_c [{shapeType=Capsule, length=2, radius=1, shapeName=dyn_hairBack_a_capsule, center=<1.2 0 0>, rotation=<rot 90 0 0>}]
 dyn_hairSide_r_c01 [{
   shapeType=Capsule
   length=2
   radius=0.75
   shapeName=dyn_hairBack_a_capsule
   center=<1.8 -1.0441 -0.3424>
   rotation=<rot 40.9925 0.001 43.9997>
  }]
 dyn_hairTop_01 [{shapeType=Capsule, length=2, radius=1, shapeName=dyn_hairBack_a_capsule, center=<3 0 0>, rotation=<rot 90 0 0>}]
 dyn_hairTop_02 [{shapeType=Capsule, length=2, radius=1, shapeName=dyn_hairBack_a_capsule, center=<1.2 0 0>, rotation=<rot 90 0 0>}]
 dyn_hairTop_a [{shapeType=Capsule, length=2, radius=1, shapeName=dyn_hairBack_a_capsule, center=<1.2 0 0>, rotation=<rot 90 0 0>}]
 dyn_hairTop_a01 [{shapeType=Capsule, length=2, radius=1, shapeName=dyn_hairBack_a_capsule, center=<1.2 0 0>, rotation=<rot 90 0 0>}]
 dyn_hairTop_b [{shapeType=Capsule, length=2, radius=1, shapeName=dyn_hairBack_a_capsule, center=<1.2 0 0>, rotation=<rot 90 0 0>}]
 dyn_hairTop_b01 [{shapeType=Capsule, length=2, radius=1, shapeName=dyn_hairBack_a_capsule, center=<1.2 0 0>, rotation=<rot 90 0 0>}]
 dyn_hairTop_c [{shapeType=Capsule, length=2, radius=1, shapeName=dyn_hairBack_a_capsule, center=<1.2 0 0>, rotation=<rot 90 0 0>}]
 dyn_hairTop_c01 [{shapeType=Capsule, length=2, radius=1, shapeName=dyn_hairBack_a_capsule, center=<1.2 0 0>, rotation=<rot 90 0 0>}]
 dyn_hairTop_d [{shapeType=Capsule, length=2, radius=1, shapeName=dyn_hairBack_a_capsule, center=<1.2 0 0>, rotation=<rot 90 0 0>}]
 dyn_hairTop_d01 [{shapeType=Capsule, length=2, radius=1, shapeName=dyn_hairBack_a_capsule, center=<1.2 0 0>, rotation=<rot 90 0 0>}]
 dyn_hairTop_e [{shapeType=Capsule, length=2, radius=1, shapeName=dyn_hairBack_a_capsule, center=<1.2 0 0>, rotation=<rot 90 0 0>}]
 dyn_hairTop_e01 [{shapeType=Capsule, length=2, radius=1, shapeName=dyn_hairBack_a_capsule, center=<1.2 0 0>, rotation=<rot 90 0 0>}]
 dyn_hairTop_f [{shapeType=Capsule, length=2, radius=1, shapeName=dyn_hairBack_a_capsule, center=<1.2 0 0>, rotation=<rot 90 0 0>}]
 dyn_hairTop_f01 [{shapeType=Capsule, length=2, radius=1, shapeName=dyn_hairBack_a_capsule, center=<1.2 0 0>, rotation=<rot 90 0 0>}]
}
constraints [
 {bone1Name=dyn_hairSide2_l, bone2Name=dyn_hairSide1_l, swing1LimitDegrees=10, swing2LimitDegrees=10, twistLimitDegrees=0}
 {bone1Name=dyn_hairSide_l_a, bone2Name=dyn_hairSide2_l, swing1LimitDegrees=10, swing2LimitDegrees=10, twistLimitDegrees=0}
 {bone1Name=dyn_hairSide_l_a01, bone2Name=dyn_hairSide_l_a, swing1LimitDegrees=10, swing2LimitDegrees=10, twistLimitDegrees=0}
 {bone1Name=dyn_hairSide_l_b, bone2Name=dyn_hairSide2_l, swing1LimitDegrees=10, swing2LimitDegrees=10, twistLimitDegrees=0}
 {bone1Name=dyn_hairSide_l_b01, bone2Name=dyn_hairSide_l_b, swing1LimitDegrees=10, swing2LimitDegrees=10, twistLimitDegrees=0}
 {bone1Name=dyn_hairSide2_r, bone2Name=dyn_hairSide1_r, swing1LimitDegrees=10, swing2LimitDegrees=10, twistLimitDegrees=0}
 {bone1Name=dyn_hairSide_r_a, bone2Name=dyn_hairSide2_r, swing1LimitDegrees=10, swing2LimitDegrees=10, twistLimitDegrees=0}
 {bone1Name=dyn_hairSide_r_a01, bone2Name=dyn_hairSide_r_a, swing1LimitDegrees=10, swing2LimitDegrees=10, twistLimitDegrees=0}
 {bone1Name=dyn_hairSide_r_b, bone2Name=dyn_hairSide2_r, swing1LimitDegrees=10, swing2LimitDegrees=15, twistLimitDegrees=0}
 {bone1Name=dyn_hairSide_r_b01, bone2Name=dyn_hairSide_r_b, swing1LimitDegrees=10, swing2LimitDegrees=10, twistLimitDegrees=0}
 {bone1Name=dyn_hairSide_r_c, bone2Name=dyn_hairSide2_r, swing1LimitDegrees=10, swing2LimitDegrees=10, twistLimitDegrees=0}
 {bone1Name=dyn_hairSide_r_c01, bone2Name=dyn_hairSide_r_c, swing1LimitDegrees=10, swing2LimitDegrees=10, twistLimitDegrees=0}
 {bone1Name=dyn_hairTop_02, bone2Name=dyn_hairTop_01, swing1LimitDegrees=25, swing2LimitDegrees=20, twistLimitDegrees=0}
 {bone1Name=dyn_hairTop_a, bone2Name=dyn_hairTop_02, swing1LimitDegrees=10, swing2LimitDegrees=10, twistLimitDegrees=0}
 {bone1Name=dyn_hairTop_a01, bone2Name=dyn_hairTop_a, swing1LimitDegrees=15, swing2LimitDegrees=15, twistLimitDegrees=0}
 {bone1Name=dyn_hairTop_b, bone2Name=dyn_hairTop_02, swing1LimitDegrees=10, swing2LimitDegrees=10, twistLimitDegrees=0}
 {bone1Name=dyn_hairTop_b01, bone2Name=dyn_hairTop_b, swing1LimitDegrees=15, swing2LimitDegrees=15, twistLimitDegrees=0}
 {bone1Name=dyn_hairTop_c, bone2Name=dyn_hairTop_02, swing1LimitDegrees=10, swing2LimitDegrees=10, twistLimitDegrees=0}
 {bone1Name=dyn_hairTop_c01, bone2Name=dyn_hairTop_c, swing1LimitDegrees=10, swing2LimitDegrees=10, twistLimitDegrees=0}
 {bone1Name=dyn_hairTop_d, bone2Name=dyn_hairTop_02, swing1LimitDegrees=10, swing2LimitDegrees=10, twistLimitDegrees=0}
 {bone1Name=dyn_hairTop_d01, bone2Name=dyn_hairTop_d, swing1LimitDegrees=10, swing2LimitDegrees=10, twistLimitDegrees=0}
 {bone1Name=dyn_hairTop_e, bone2Name=dyn_hairTop_02, swing1LimitDegrees=10, swing2LimitDegrees=10, twistLimitDegrees=0}
 {bone1Name=dyn_hairTop_e01, bone2Name=dyn_hairTop_e, swing1LimitDegrees=10, swing2LimitDegrees=10, twistLimitDegrees=0}
 {bone1Name=dyn_hairTop_f, bone2Name=dyn_hairTop_02, swing1LimitDegrees=10, swing2LimitDegrees=10, twistLimitDegrees=0}
 {bone1Name=dyn_hairTop_f01, bone2Name=dyn_hairTop_f, swing1LimitDegrees=10, swing2LimitDegrees=10, twistLimitDegrees=0}
 {bone1Name=dyn_hairSide1_l, bone2Name=base_hairSide1_l, swing1LimitDegrees=10, swing2LimitDegrees=10, twistLimitDegrees=0}
 {bone1Name=dyn_hairSide1_r, bone2Name=base_hairSide1_r, swing1LimitDegrees=5, swing2LimitDegrees=8, twistLimitDegrees=0}
 {bone1Name=dyn_hairBack_a, bone2Name=dyn_hairBack, swing1LimitDegrees=10, swing2LimitDegrees=10, twistLimitDegrees=0}
 {bone1Name=dyn_hairBack_a1, bone2Name=dyn_hairBack_a, swing1LimitDegrees=10, swing2LimitDegrees=10, twistLimitDegrees=0}
 {bone1Name=dyn_hairBack_b, bone2Name=dyn_hairBack, swing1LimitDegrees=10, swing2LimitDegrees=10, twistLimitDegrees=0}
 {bone1Name=dyn_hairBack_c, bone2Name=dyn_hairBack, swing1LimitDegrees=10, swing2LimitDegrees=10, twistLimitDegrees=0}
 {bone1Name=dyn_hairBack_d, bone2Name=dyn_hairBack, swing1LimitDegrees=10, swing2LimitDegrees=10, twistLimitDegrees=0}
 {bone1Name=dyn_hairBack_e, bone2Name=base_hairBack_e, swing1LimitDegrees=10, swing2LimitDegrees=10, twistLimitDegrees=0}
 {bone1Name=dyn_hairBack_l_e1, bone2Name=dyn_hairBack_e, swing1LimitDegrees=20, swing2LimitDegrees=30, twistLimitDegrees=0}
 {bone1Name=dyn_hairBack_b1, bone2Name=dyn_hairBack_b, swing1LimitDegrees=10, swing2LimitDegrees=10, twistLimitDegrees=0}
 {bone1Name=dyn_hairBack_c1, bone2Name=dyn_hairBack_c, swing1LimitDegrees=10, swing2LimitDegrees=10, twistLimitDegrees=0}
 {bone1Name=dyn_hairBack_d1, bone2Name=dyn_hairBack_d, swing1LimitDegrees=10, swing2LimitDegrees=10, twistLimitDegrees=0}
 {bone1Name=dyn_hairTop_01, bone2Name=base_hairTop, swing1LimitDegrees=30, swing2LimitDegrees=40, twistLimitDegrees=0}
 {bone1Name=dyn_hairBack, bone2Name=base_hairBack, swing1LimitDegrees=20, swing2LimitDegrees=15, twistLimitDegrees=0}
]
