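# Post-process style AnimBP for the GroovyReader face accessory, animating hair bones that follow the attached parent mesh.
# Copied pose runs through a crouch-aware Control Rig, then RigidBody simulation with state-driven gravity and skydive-float sim-space settings.
# Cached-pose stages add sine-wave sway while sprinting grounded, then alpha-gated bone corrections for crouch, skydive, parachute, backpack, falling and DBNO.
# The update event derives alphas, gravity and wave rotators from movement and backbling state.
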AnimBlueprint /BRCosmetics/Characters/Player/Female/Medium/Bodies/F_MED_GroovyReader/Meshes/Parts/F_MED_GroovyReader_FaceAcc_AnimBP
target_skeleton=&/F_MED_GroovyReader_FaceAcc_Skeleton
graphs [
 AnimGraph {nodes [
   n1 = "Transform (Modify) Bone - Bone: dyn_hairTop_01"(
    ComponentPose=Local To Component(LocalPose=Use cached pose 'Sprinting Fix cache'())
    Alpha=Get Alpha_isDBNO() -> (Alpha_isDBNO:double)
    Translation="0,0,0"
    Rotation="12,22,12"
    Scale="1,1,1"
   ) {
    rotationMode=BMM_Additive
    rotationSpace=BCS_BoneSpace
    translationMode=BMM_Additive
    translationSpace=BCS_BoneSpace
    alphaBoolBlend {blendOption=Linear}
    alphaScaleBiasClamp {bInterpResult=true}
    boneToModify {boneName=dyn_hairTop_01}
    rotation=<rot 12 22 12>
   }
   n2 = "Transform (Modify) Bone - Bone: dyn_hairTop_e"(ComponentPose=@n1, Alpha=Get Alpha_isDBNOIdle() -> (Alpha_isDBNOIdle:double), Translation="0,0,0", Rotation="19,0,0", Scale="1,1,1") {
    rotationMode=BMM_Additive
    rotationSpace=BCS_BoneSpace
    translationMode=BMM_Additive
    translationSpace=BCS_BoneSpace
    alphaBoolBlend {blendOption=Linear}
    alphaScaleBiasClamp {bInterpResult=true}
    boneToModify {boneName=dyn_hairTop_e}
    rotation {pitch=19}
   }
   Output Pose(Result=Component To Local(ComponentPose=@n2)) {layerGroup=None}
   n3 = Copy Pose From Mesh() {bUseAttachedParent=true}
   n4 = Control Rig!(
    Source=Component To Local(ComponentPose=Local To Component(LocalPose=@n3))
    Alpha
    isCrouching:double=Get Alpha_isCrouching() -> (Alpha_isCrouching:double)
   ) -> (Pose) {
    alphaBoolBlend {blendOption=Linear}
    alphaScaleBiasClamp {bInterpResult=true, interpSpeedDecreasing=2, interpSpeedIncreasing=3}
    controlRigAssetReference {blueprintRigClass=&/F_MED_GroovyReader_FaceAcc_CtrlRig.F_MED_GroovyReader_FaceAcc_CtrlRig_C}
   }
   Save cached pose 'ctrlRig Fix cache'(Pose=@n4) {cacheName=ctrlRig Fix cache}
   n5 = Get rotatorWaveHairTop() -> (rotatorWaveHairTop:Rotator)
   n6 = BreakRotator(InRot=@n5)
   n7 = SelectRotator!(
    A:Rotator=MakeRotator(Pitch=float * float(A=@n6.Pitch, B=0.7) {is=title K2Node_PromotableOperator})
    B:Rotator=@n5
    bPickA:bool=Get bIsCrouchSprinting()
   ) -> (ReturnValue:Rotator)
   n8 = "Transform (Modify) Bone - Bone: base_hairTop"(
    ComponentPose=Local To Component(LocalPose=Use cached pose 'RigidBody Fix cache'())
    Alpha=1
    Translation="0,0,0"
    Rotation=@n7
    Scale="1,1,1"
   ) {
    rotationMode=BMM_Additive
    rotationSpace=BCS_BoneSpace
    translationMode=BMM_Additive
    translationSpace=BCS_BoneSpace
    alphaBoolBlend {blendOption=Linear}
    alphaScaleBiasClamp {interpSpeedDecreasing=2, interpSpeedIncreasing=1}
    boneToModify {boneName=base_hairTop}
    rotation {yaw=90}
   }
   n9 = Get rotatorWaveHairBack() -> (rotatorWaveHairBack:Rotator)
   n10 = "Transform (Modify) Bone - Bone: base_hairBack"(ComponentPose=@n8, Alpha=1, Translation="0,0,0", Rotation=@n9, Scale="1,1,1") {
    rotationMode=BMM_Additive
    rotationSpace=BCS_BoneSpace
    translationMode=BMM_Additive
    translationSpace=BCS_BoneSpace
    alphaBoolBlend {blendOption=Linear}
    alphaScaleBiasClamp {interpSpeedDecreasing=2, interpSpeedIncreasing=1}
    boneToModify {boneName=base_hairBack}
    rotation {yaw=90}
   }
   n11 = Component To Local(ComponentPose=@n10)
   n12 = Use cached pose 'RigidBody Fix cache'()
   n13 = Get SprintingNotJumping() -> (SprintingNotJumping:bool)
   n14 = "Transform (Modify) Bone - Bone: base_hairTop"(
    ComponentPose=Local To Component(LocalPose=Use cached pose 'RigidBody Fix cache'())
    Alpha=1
    Translation="0,0,0"
    Rotation=@n7
    Scale="1,1,1"
   ) {
    rotationMode=BMM_Additive
    rotationSpace=BCS_BoneSpace
    translationMode=BMM_Additive
    translationSpace=BCS_BoneSpace
    alphaBoolBlend {blendOption=Linear}
    alphaScaleBiasClamp {interpSpeedDecreasing=2, interpSpeedIncreasing=1}
    boneToModify {boneName=base_hairTop}
    rotation {yaw=90}
   }
   n15 = "Transform (Modify) Bone - Bone: base_hairBack"(ComponentPose=@n14, Alpha=1, Translation="0,0,0", Rotation=@n9, Scale="1,1,1") {
    rotationMode=BMM_Additive
    rotationSpace=BCS_BoneSpace
    translationMode=BMM_Additive
    translationSpace=BCS_BoneSpace
    alphaBoolBlend {blendOption=Linear}
    alphaScaleBiasClamp {interpSpeedDecreasing=2, interpSpeedIncreasing=1}
    boneToModify {boneName=base_hairBack}
    rotation {yaw=90}
   }
   Save cached pose 'SineFX Cache'(Pose=Blend Poses by bool(
     BlendPose_0=Component To Local(ComponentPose=@n15)
     BlendPose_1=Blend Poses by bool(BlendPose_0=@n11, BlendPose_1=@n12, BlendTime_0=0.2, BlendTime_1=0.2, bActiveValue=@n13) {blendType=Linear}
     BlendTime_0=0.2
     BlendTime_1=0.2
     bActiveValue=Get TacticalSprintingNotJumping() -> (TacticalSprintingNotJumping:bool)
    ) {blendType=Linear}) {cacheName=SineFX Cache}
   n16 = "Transform (Modify) Bone - Bone: base_hairTop"(
    ComponentPose=Local To Component(LocalPose=Use cached pose 'SineFX Cache'())
    Alpha=Get Alpha_isDisableRigbodySprinting() -> (Alpha_isDisableRigbodySprinting:double)
    Translation="0,0,0"
    Rotation="30,0,0"
    Scale="1,1,1"
   ) {
    rotationMode=BMM_Additive
    rotationSpace=BCS_BoneSpace
    translationMode=BMM_Additive
    translationSpace=BCS_BoneSpace
    alphaBoolBlend {blendOption=Linear}
    alphaScaleBiasClamp {bInterpResult=true, interpSpeedDecreasing=5, interpSpeedIncreasing=5}
    boneToModify {boneName=base_hairTop}
    rotation {pitch=30}
   }
   n17 = "Constraint - base_hairBack_e (P)"(ComponentPose=@n16, Alpha=1, ConstraintWeights_0=1) {
    alphaBoolBlend {blendOption=Linear}
    boneToModify {boneName=base_hairBack_e}
    constraintSetup [{targetBone {boneName=dyn_hairBack}}]
    constraintWeights [1]
   }
   n18 = Get Alpha_moveSideCrouch() -> (Alpha_moveSideCrouch:float)
   n19 = "Transform (Modify) Bone - Bone: dyn_hairTop_f"(ComponentPose=@n17, Alpha=@n18, Translation="0,0,0", Rotation="22,0,0", Scale="1,1,1") {
    rotationMode=BMM_Replace
    rotationSpace=BCS_BoneSpace
    translationMode=BMM_Replace
    translationSpace=BCS_BoneSpace
    alphaBoolBlend {blendOption=Linear}
    alphaScaleBiasClamp {bInterpResult=true, interpSpeedDecreasing=5, interpSpeedIncreasing=5}
    boneToModify {boneName=dyn_hairTop_f}
    rotation {pitch=22}
   }
   n20 = "Transform (Modify) Bone - Bone: dyn_hairTop_e"(ComponentPose=@n19, Alpha=@n18, Translation="0,0,0", Rotation="10,0,0", Scale="1,1,1") {
    rotationMode=BMM_Replace
    rotationSpace=BCS_BoneSpace
    translationMode=BMM_Replace
    translationSpace=BCS_BoneSpace
    alphaBoolBlend {blendOption=Linear}
    alphaScaleBiasClamp {bInterpResult=true, interpSpeedDecreasing=5, interpSpeedIncreasing=5}
    boneToModify {boneName=dyn_hairTop_e}
    rotation {pitch=10}
   }
   n21 = "Transform (Modify) Bone - Bone: dyn_hairTop_a"(ComponentPose=@n20, Alpha=@n18, Translation="0,0,0", Rotation="15,0,0", Scale="1,1,1") {
    rotationMode=BMM_Replace
    rotationSpace=BCS_BoneSpace
    translationMode=BMM_Replace
    translationSpace=BCS_BoneSpace
    alphaBoolBlend {blendOption=Linear}
    alphaScaleBiasClamp {bInterpResult=true, interpSpeedDecreasing=5, interpSpeedIncreasing=5}
    boneToModify {boneName=dyn_hairTop_a}
    rotation {pitch=15}
   }
   n22 = "Transform (Modify) Bone - Bone: dyn_hairTop_02"(ComponentPose=@n21, Alpha=@n18, Translation="0,0,0", Rotation="0,25,0", Scale="1,1,1") {
    rotationMode=BMM_Replace
    rotationSpace=BCS_BoneSpace
    translationMode=BMM_Replace
    translationSpace=BCS_BoneSpace
    alphaBoolBlend {blendOption=Linear}
    alphaScaleBiasClamp {bInterpResult=true, interpSpeedDecreasing=5, interpSpeedIncreasing=5}
    boneToModify {boneName=dyn_hairTop_02}
    rotation {yaw=25}
   }
   n23 = "Transform (Modify) Bone - Bone: dyn_hairTop_d"(ComponentPose=@n22, Alpha=@n18, Translation="0,0,0", Rotation="20,3,0", Scale="1,1,1") {
    rotationMode=BMM_Replace
    rotationSpace=BCS_BoneSpace
    translationMode=BMM_Replace
    translationSpace=BCS_BoneSpace
    alphaBoolBlend {blendOption=Linear}
    alphaScaleBiasClamp {bInterpResult=true, interpSpeedDecreasing=5, interpSpeedIncreasing=5}
    boneToModify {boneName=dyn_hairTop_d}
    rotation {pitch=20, yaw=3}
   }
   n24 = "Transform (Modify) Bone - Bone: dyn_hairTop_c"(ComponentPose=@n23, Alpha=@n18, Translation="0,0,0", Rotation="8,5,0", Scale="1,1,1") {
    rotationMode=BMM_Replace
    rotationSpace=BCS_BoneSpace
    translationMode=BMM_Replace
    translationSpace=BCS_BoneSpace
    alphaBoolBlend {blendOption=Linear}
    alphaScaleBiasClamp {bInterpResult=true, interpSpeedDecreasing=5, interpSpeedIncreasing=5}
    boneToModify {boneName=dyn_hairTop_c}
    rotation {pitch=8, yaw=5}
   }
   Save cached pose 'Side Crouch Fix cache'(Pose=Component To Local(ComponentPose=@n24)) {cacheName=Side Crouch Fix cache}
   n25 = Get Alpha_sprintCrouch() -> (Alpha_sprintCrouch:float)
   n26 = "Transform (Modify) Bone - Bone: base_hairTop"(
    ComponentPose=Local To Component(LocalPose=Use cached pose 'Side Crouch Fix cache'())
    Alpha=@n25
    Translation="0,0,0"
    Rotation="17,0,5"
    Scale="1,1,1"
   ) {
    rotationMode=BMM_Additive
    rotationSpace=BCS_BoneSpace
    translationMode=BMM_Additive
    translationSpace=BCS_BoneSpace
    alphaBoolBlend {blendOption=Linear}
    alphaScaleBiasClamp {bInterpResult=true}
    boneToModify {boneName=base_hairTop}
    rotation {pitch=17, roll=5}
   }
   n27 = "Transform (Modify) Bone - Bone: dyn_hairBack"(ComponentPose=@n26, Alpha=@n25, Translation="0,0,0", Rotation="-2,0,0", Scale="1,1,1") {
    rotationMode=BMM_Replace
    rotationSpace=BCS_BoneSpace
    translationMode=BMM_Replace
    translationSpace=BCS_BoneSpace
    alphaBoolBlend {blendOption=Linear}
    alphaScaleBiasClamp {bInterpResult=true, interpSpeedDecreasing=5, interpSpeedIncreasing=5}
    boneToModify {boneName=dyn_hairBack}
    rotation {pitch=-2}
   }
   n28 = "Transform (Modify) Bone - Bone: dyn_hairTop_e"(ComponentPose=@n27, Alpha=@n25, Translation="0,0,0", Rotation="5,0,0", Scale="1,1,1") {
    rotationMode=BMM_Replace
    rotationSpace=BCS_BoneSpace
    translationMode=BMM_Replace
    translationSpace=BCS_BoneSpace
    alphaBoolBlend {blendOption=Linear}
    alphaScaleBiasClamp {bInterpResult=true, interpSpeedDecreasing=5, interpSpeedIncreasing=5}
    boneToModify {boneName=dyn_hairTop_e}
    rotation {pitch=5}
   }
   Save cached pose 'SprintCrouchFixCache'(Pose=Component To Local(ComponentPose=@n28)) {cacheName=SprintCrouchFixCache}
   n29 = Get Alpha_SkyDiveFloating() -> (Alpha_SkyDiveFloating:float)
   n30 = "Transform (Modify) Bone - Bone: dyn_hairSide2_l"(
    ComponentPose=Local To Component(LocalPose=Use cached pose 'SprintCrouchFixCache'())
    Alpha=@n29
    Translation="0,0,0"
    Rotation="-20,-40,0"
    Scale="1,1,1"
   ) {
    rotationMode=BMM_Additive
    rotationSpace=BCS_BoneSpace
    translationMode=BMM_Additive
    translationSpace=BCS_BoneSpace
    alphaBoolBlend {blendOption=Linear}
    alphaScaleBiasClamp {bInterpResult=true}
    boneToModify {boneName=dyn_hairSide2_l}
    rotation {pitch=-20, yaw=-40}
   }
   n31 = "Transform (Modify) Bone - Bone: base_hairBack_extra"(ComponentPose=@n30, Alpha=@n29, Translation="-5,0,5", Rotation="0,0,0", Scale="1,1,1") {
    rotationSpace=BCS_BoneSpace
    scaleMode=BMM_Additive
    scaleSpace=BCS_BoneSpace
    translationMode=BMM_Additive
    translationSpace=BCS_BoneSpace
    alphaBoolBlend {blendOption=Linear}
    alphaScaleBiasClamp {bInterpResult=true}
    boneToModify {boneName=base_hairBack_extra}
    translation {x=-5, z=5}
   }
   n32 = "Transform (Modify) Bone - Bone: dyn_hairSide2_r"(ComponentPose=@n31, Alpha=@n29, Translation="0,0,0", Rotation="-20,15,0", Scale="1,1,1") {
    rotationMode=BMM_Additive
    rotationSpace=BCS_BoneSpace
    translationMode=BMM_Additive
    translationSpace=BCS_BoneSpace
    alphaBoolBlend {blendOption=Linear}
    alphaScaleBiasClamp {bInterpResult=true}
    boneToModify {boneName=dyn_hairSide2_r}
    rotation {pitch=-20, yaw=15}
   }
   n33 = "Transform (Modify) Bone - Bone: dyn_hairSide_l_b"(ComponentPose=@n32, Alpha=@n29, Translation="0,0,0", Rotation="7,-20,0", Scale="1,1,1") {
    rotationMode=BMM_Additive
    rotationSpace=BCS_BoneSpace
    translationMode=BMM_Additive
    translationSpace=BCS_BoneSpace
    alphaBoolBlend {blendOption=Linear}
    alphaScaleBiasClamp {bInterpResult=true}
    boneToModify {boneName=dyn_hairSide_l_b}
    rotation {pitch=7, yaw=-20}
   }
   n34 = "Transform (Modify) Bone - Bone: dyn_hairBack"(ComponentPose=@n33, Alpha=@n29, Translation="0,0,0", Rotation="-30,0,0", Scale="1,1,1") {
    rotationMode=BMM_Additive
    rotationSpace=BCS_BoneSpace
    translationMode=BMM_Additive
    translationSpace=BCS_BoneSpace
    alphaBoolBlend {blendOption=Linear}
    alphaScaleBiasClamp {bInterpResult=true}
    boneToModify {boneName=dyn_hairBack}
    rotation {pitch=-30}
   }
   n35 = "Transform (Modify) Bone - Bone: dyn_hairSide_r_a"(ComponentPose=@n34, Alpha=@n29, Translation="0,0,0", Rotation="0,-20,0", Scale="1,1,1") {
    rotationMode=BMM_Additive
    rotationSpace=BCS_BoneSpace
    translationMode=BMM_Additive
    translationSpace=BCS_BoneSpace
    alphaBoolBlend {blendOption=Linear}
    alphaScaleBiasClamp {bInterpResult=true}
    boneToModify {boneName=dyn_hairSide_r_a}
    rotation {yaw=-20}
   }
   n36 = "Transform (Modify) Bone - Bone: dyn_hairBack_d"(ComponentPose=@n35, Alpha=@n29, Translation="0,0,0", Rotation="35,0,0", Scale="1,1,1") {
    rotationMode=BMM_Additive
    rotationSpace=BCS_BoneSpace
    translationMode=BMM_Additive
    translationSpace=BCS_BoneSpace
    alphaBoolBlend {blendOption=Linear}
    alphaScaleBiasClamp {bInterpResult=true}
    boneToModify {boneName=dyn_hairBack_d}
    rotation {pitch=35}
   }
   n37 = Get Alpha_SkydiveMode() -> (Alpha_SkydiveMode:double)
   n38 = "Transform (Modify) Bone - Bone: dyn_hairTop_01"(ComponentPose=@n36, Alpha=@n37, Translation="0,0,0", Rotation="-40,0,0", Scale="1,1,1") {
    rotationMode=BMM_Additive
    rotationSpace=BCS_BoneSpace
    translationMode=BMM_Additive
    translationSpace=BCS_BoneSpace
    alphaBoolBlend {blendOption=Linear}
    alphaScaleBiasClamp {bInterpResult=true}
    boneToModify {boneName=dyn_hairTop_01}
    rotation {pitch=-40}
   }
   n39 = "Transform (Modify) Bone - Bone: dyn_hairTop_02"(ComponentPose=@n38, Alpha=@n37, Translation="0,0,0", Rotation="0,-20,0", Scale="1,1,1") {
    rotationMode=BMM_Additive
    rotationSpace=BCS_BoneSpace
    translationMode=BMM_Additive
    translationSpace=BCS_BoneSpace
    alphaBoolBlend {blendOption=Linear}
    alphaScaleBiasClamp {bInterpResult=true}
    boneToModify {boneName=dyn_hairTop_02}
    rotation {yaw=-20}
   }
   n40 = "Transform (Modify) Bone - Bone: dyn_hairBack"(ComponentPose=@n39, Alpha=Get Alpha_isUseBackpack() -> (Alpha_isUseBackpack:double), Translation="0,0,0", Rotation="35,0,0", Scale="1,1,1") {
    rotationMode=BMM_Additive
    rotationSpace=BCS_BoneSpace
    translationMode=BMM_Additive
    translationSpace=BCS_BoneSpace
    alphaBoolBlend {blendOption=Linear}
    alphaScaleBiasClamp {bInterpResult=true}
    boneToModify {boneName=dyn_hairBack}
    rotation {pitch=35}
   }
   Save cached pose 'SkyDiveFloating Fix Cache'(Pose=Component To Local(ComponentPose=@n40)) {cacheName=SkyDiveFloating Fix Cache}
   n41 = Get Alpha_isOpenParachute() -> (Alpha_isOpenParachute:double)
   n42 = "Transform (Modify) Bone - Bone: dyn_hairSide1_r"(
    ComponentPose=Local To Component(LocalPose=Use cached pose 'SkyDiveFloating Fix Cache'())
    Alpha=@n41
    Translation="0,0,0"
    Rotation="-10,40,0"
    Scale="1,1,1"
   ) {
    rotationMode=BMM_Additive
    rotationSpace=BCS_BoneSpace
    translationMode=BMM_Additive
    translationSpace=BCS_BoneSpace
    alphaBoolBlend {blendOption=Linear}
    alphaScaleBiasClamp {bInterpResult=true}
    boneToModify {boneName=dyn_hairSide1_r}
    rotation {pitch=-10, yaw=40}
   }
   n43 = "Transform (Modify) Bone - Bone: dyn_hairSide_r_b"(ComponentPose=@n42, Alpha=@n41, Translation="0,0,0", Rotation="-15,6,0", Scale="1,1,1") {
    rotationMode=BMM_Additive
    rotationSpace=BCS_BoneSpace
    translationMode=BMM_Additive
    translationSpace=BCS_BoneSpace
    alphaBoolBlend {blendOption=Linear}
    alphaScaleBiasClamp {bInterpResult=true}
    boneToModify {boneName=dyn_hairSide_r_b}
    rotation {pitch=-15, yaw=6}
   }
   n44 = "Transform (Modify) Bone - Bone: dyn_hairSide1_l"(ComponentPose=@n43, Alpha=@n41, Translation="0,0,0", Rotation="0,-30,0", Scale="1,1,1") {
    rotationMode=BMM_Additive
    rotationSpace=BCS_BoneSpace
    translationMode=BMM_Additive
    translationSpace=BCS_BoneSpace
    alphaBoolBlend {blendOption=Linear}
    alphaScaleBiasClamp {bInterpResult=true}
    boneToModify {boneName=dyn_hairSide1_l}
    rotation {yaw=-30}
   }
   n45 = "Transform (Modify) Bone - Bone: dyn_hairSide_l_a"(ComponentPose=@n44, Alpha=@n41, Translation="0,0,0", Rotation="0,-8,0", Scale="1,1,1") {
    rotationMode=BMM_Additive
    rotationSpace=BCS_BoneSpace
    translationMode=BMM_Additive
    translationSpace=BCS_BoneSpace
    alphaBoolBlend {blendOption=Linear}
    alphaScaleBiasClamp {bInterpResult=true}
    boneToModify {boneName=dyn_hairSide_l_a}
    rotation {yaw=-8}
   }
   n46 = "Transform (Modify) Bone - Bone: dyn_hairBack_e"(ComponentPose=@n45, Alpha=@n41, Translation="0,0,0", Rotation="25,-5,0", Scale="1,1,1") {
    rotationMode=BMM_Additive
    rotationSpace=BCS_BoneSpace
    translationMode=BMM_Additive
    translationSpace=BCS_BoneSpace
    alphaBoolBlend {blendOption=Linear}
    alphaScaleBiasClamp {bInterpResult=true}
    boneToModify {boneName=dyn_hairBack_e}
    rotation {pitch=25, yaw=-5}
   }
   n47 = "Transform (Modify) Bone - Bone: dyn_hairBack_l_e1"(ComponentPose=@n46, Alpha=@n41, Translation="0,0,0", Rotation="15,0,0", Scale="1,1,1") {
    rotationMode=BMM_Additive
    rotationSpace=BCS_BoneSpace
    translationMode=BMM_Additive
    translationSpace=BCS_BoneSpace
    alphaBoolBlend {blendOption=Linear}
    alphaScaleBiasClamp {bInterpResult=true}
    boneToModify {boneName=dyn_hairBack_l_e1}
    rotation {pitch=15}
   }
   n48 = "Transform (Modify) Bone - Bone: dyn_hairBack"(ComponentPose=@n47, Alpha=@n41, Translation="0,0,0", Rotation="15,0,0", Scale="1,1,1") {
    rotationMode=BMM_Additive
    rotationSpace=BCS_BoneSpace
    translationMode=BMM_Additive
    translationSpace=BCS_BoneSpace
    alphaBoolBlend {blendOption=Linear}
    alphaScaleBiasClamp {bInterpResult=true}
    boneToModify {boneName=dyn_hairBack}
    rotation {pitch=15}
   }
   n49 = "Transform (Modify) Bone - Bone: dyn_hairTop_01"(ComponentPose=@n48, Alpha=@n41, Translation="0,0,0", Rotation="7,0,0", Scale="1,1,1") {
    rotationMode=BMM_Additive
    rotationSpace=BCS_BoneSpace
    translationMode=BMM_Additive
    translationSpace=BCS_BoneSpace
    alphaBoolBlend {blendOption=Linear}
    alphaScaleBiasClamp {bInterpResult=true}
    boneToModify {boneName=dyn_hairTop_01}
    rotation {pitch=7}
   }
   n50 = "Transform (Modify) Bone - Bone: base_hairSide1_r"(ComponentPose=@n49, Alpha=@n41, Translation="0,0,0", Rotation="0,-20,0", Scale="1,1,1") {
    rotationMode=BMM_Additive
    rotationSpace=BCS_BoneSpace
    translationMode=BMM_Additive
    translationSpace=BCS_BoneSpace
    alphaBoolBlend {blendOption=Linear}
    alphaScaleBiasClamp {bInterpResult=true}
    boneToModify {boneName=base_hairSide1_r}
    rotation {yaw=-20}
   }
   Save cached pose 'OpenParachute Fix Cache'(Pose=Component To Local(ComponentPose=@n50)) {cacheName=OpenParachute Fix Cache}
   n51 = Get Alpha_isSprinting() -> (Alpha_isSprinting:double)
   n52 = "Transform (Modify) Bone - Bone: dyn_hairTop_02"(
    ComponentPose=Local To Component(LocalPose=Use cached pose 'OpenParachute Fix Cache'())
    Alpha=@n51
    Translation="0,0,0"
    Rotation="23,0,0"
    Scale="1,1,1"
   ) {
    rotationMode=BMM_Additive
    rotationSpace=BCS_BoneSpace
    translationMode=BMM_Additive
    translationSpace=BCS_BoneSpace
    alphaBoolBlend {blendOption=Linear}
    alphaScaleBiasClamp {bInterpResult=true}
    boneToModify {boneName=dyn_hairTop_02}
    rotation {pitch=23}
   }
   n53 = "Transform (Modify) Bone - Bone: dyn_hairTop_b"(ComponentPose=@n52, Alpha=@n51, Translation="0,0,0", Rotation="15,0,0", Scale="1,1,1") {
    rotationMode=BMM_Additive
    rotationSpace=BCS_BoneSpace
    translationMode=BMM_Additive
    translationSpace=BCS_BoneSpace
    alphaBoolBlend {blendOption=Linear}
    alphaScaleBiasClamp {bInterpResult=true}
    boneToModify {boneName=dyn_hairTop_b}
    rotation {pitch=15}
   }
   n54 = Get Alpha_isFalling() -> (Alpha_isFalling:double)
   n55 = "Transform (Modify) Bone - Bone: dyn_hairTop_01"(ComponentPose=@n53, Alpha=@n54, Translation="0,0,0", Rotation="30,0,0", Scale="1,1,1") {
    rotationMode=BMM_Additive
    rotationSpace=BCS_BoneSpace
    translationMode=BMM_Additive
    translationSpace=BCS_BoneSpace
    alphaBoolBlend {blendOption=Linear}
    alphaScaleBiasClamp {bInterpResult=true, interpSpeedDecreasing=3, interpSpeedIncreasing=1}
    boneToModify {boneName=dyn_hairTop_01}
    rotation {pitch=30}
   }
   n56 = "Transform (Modify) Bone - Bone: dyn_hairTop_02"(ComponentPose=@n55, Alpha=@n54, Translation="0,0,0", Rotation="0,15,0", Scale="1,1,1") {
    rotationMode=BMM_Additive
    rotationSpace=BCS_BoneSpace
    translationMode=BMM_Additive
    translationSpace=BCS_BoneSpace
    alphaBoolBlend {blendOption=Linear}
    alphaScaleBiasClamp {bInterpResult=true, interpSpeedDecreasing=5, interpSpeedIncreasing=5}
    boneToModify {boneName=dyn_hairTop_02}
    rotation {yaw=15}
   }
   Save cached pose 'Sprinting Fix cache'(Pose=Component To Local(ComponentPose=@n56)) {cacheName=Sprinting Fix cache}
   n57 = RigidBody!(
    ComponentPose=Local To Component(LocalPose=Use cached pose 'ctrlRig Fix cache'())
    Alpha
    OverrideWorldGravity:Vector=Get GravityOverride()
    ExternalForce
    SimSpaceSettings:Sim Space Settings Structure=Get Sim Space Settings() -> (Sim Space Settings:Sim Space Settings Structure)
   ) -> (Pose) {
    bOverrideWorldGravity=true
    overridePhysicsAsset=&/F_MED_GroovyReader_FaceAcc_Physics
    simulationSpace=BaseBoneSpace
    alphaBoolBlend {blendOption=Linear}
    baseBoneRef {boneName=neck_01}
    componentAppliedLinearAccClamp=<3000 3000 3000>
    componentLinearAccScale=<0.1 0.1 0.1>
    componentLinearVelScale=<0.2 0.2 0.2>
    simSpaceSettings {
     maxAngularAcceleration=1000
     maxAngularVelocity=3
     maxLinearAcceleration=700
     maxLinearVelocity=500
     velocityScaleZ=0
     worldAlpha=0.5
     externalLinearDragV=<0.2 0.2>
    }
   }
   n58 = "Transform (Modify) Bone - Bone: base_hairTop"(
    ComponentPose=@n57
    Alpha=Get Alpha_isDisableRigbody() -> (Alpha_isDisableRigbody:double)
    Translation="0,0,0"
    Rotation="0,0,0"
    Scale="1,1,1"
   ) {
    rotationMode=BMM_Additive
    rotationSpace=BCS_BoneSpace
    translationMode=BMM_Additive
    translationSpace=BCS_BoneSpace
    alphaBoolBlend {blendOption=Linear}
    alphaScaleBiasClamp {bInterpResult=true, interpSpeedDecreasing=5, interpSpeedIncreasing=5}
    boneToModify {boneName=base_hairTop}
   }
   Save cached pose 'RigidBody Fix cache'(Pose=Component To Local(ComponentPose=@n58)) {cacheName=RigidBody Fix cache}
   Get Alpha_isTargeting() -> (Alpha_isTargeting:double)
  ]}
 EventGraph {nodes [
   Event BlueprintUpdateAnimation()
   > n1 = Sequence!(execute) -> (then_0, then_1, then_2:Exec, then_3:Exec)
   > n2 = Sequence!(execute) -> (then_0, then_1, then_2, then_3, then_4:Exec, then_5:Exec, then_6:Exec)
   n3 = Branch(execute=@n2.then_1, Condition=Get bIsSkydiveFloating())
   > Set Alpha_SkyDiveFloating(Alpha_SkyDiveFloating:float=1) -> (Output_Get:float)
   Set Alpha_SkyDiveFloating(execute=@n3.else, Alpha_SkyDiveFloating=0)
   n4 = Get bIsSprinting()
   n5 = Get bIsCrouching()
   n6 = Branch(execute=@n2.then_2, Condition=AND Boolean(A=@n4, B=@n5))
   > Set Alpha_sprintCrouch(Alpha_sprintCrouch:float=1) -> (Output_Get:float)
   Set Alpha_sprintCrouch(execute=@n6.else, Alpha_sprintCrouch=0)
   n7 = Branch(execute=@n2.then_4, Condition=Get bIsSprinting())
   > Set Alpha_isSprinting(Alpha_isSprinting:double=1) -> (Output_Get:double)
   Set Alpha_isSprinting(execute=@n7.else, Alpha_isSprinting=0)
   n8 = Branch(execute=@n2.then_6, Condition=Get bIsCrouching())
   > Set Alpha_isCrouching(Alpha_isCrouching:double=1) -> (Output_Get:double)
   Set Alpha_isCrouching(execute=@n8.else, Alpha_isCrouching=0)
   n9 = Gravity Override State Function DEPRECATED(
    execute=@n2.then_5
    CharacterPartAnimInstance=Self-Reference()
    GameTimeInSeconds=GetGameTimeInSeconds()
    GravityOverrideParams=Get Gravity Override Params()
   )
   > Set GravityOverride(GravityOverride=@n9)
   n10 = NOT Boolean(A=@n4)
   n11 = Get bIsMoving2D()
   n12 = Branch(execute=@n2.then_3, Condition=AND Boolean(A=@n5, B=AND Boolean(A=@n10, B=@n11)))
   > Set Alpha_moveSideCrouch(Alpha_moveSideCrouch:float=1) -> (Output_Get:float)
   Set Alpha_moveSideCrouch(execute=@n12.else, Alpha_moveSideCrouch=0)
   n13 = Branch(execute=@n2.then_0, Condition=Get bIsSkydiveDiveMode())
   > Set Alpha_SkydiveMode(Alpha_SkydiveMode:double=1) -> (Output_Get:double)
   Set Alpha_SkydiveMode(execute=@n13.else, Alpha_SkydiveMode=0)
   n14 = NOR Boolean!(A:bool=Get bIsJumping(), B:bool=Get bIsFalling()) -> (ReturnValue:bool)
   Set SprintingNotJumping(execute=@n1.then_1, SprintingNotJumping:bool=AND Boolean(A=Get bIsSprinting(), B=@n14)) -> (Output_Get:bool)
   > Set TacticalSprintingNotJumping(TacticalSprintingNotJumping:bool=AND Boolean(A=@n14, B=Get bIsTacticalSprinting())) -> (Output_Get:bool)
   n15 = Sequence!(execute=@n1.then_3) -> (then_0, then_1, then_2, then_3, then_4, then_5, then_6, then_7:Exec, then_8:Exec, then_9:Exec, then_10:Exec, then_11:Exec)
   n16 = float * float(A=GetRealTimeSeconds(), B=Get speedWaveHairTop() -> (speedWaveHairTop:double))
   n17 = float * float(A=MultiplyByPi!(Value:double="Sin (Radians)"(A=@n16)) -> (ReturnValue:double), B=Get amplitudWaveHairTop() -> (amplitudWaveHairTop:double))
   n18 = Get offsetWaveHairTop() -> (offsetWaveHairTop:double)
   Set rotatorWaveHairTop(execute=@n15.then_1, rotatorWaveHairTop:Rotator=MakeRotator(Pitch=float + float(A=@n17, B=@n18))) -> (Output_Get:Rotator)
   n19 = Branch(execute=@n15.then_3, Condition=Get bIsTargeting())
   > Set Alpha_isTargeting(Alpha_isTargeting:double=1) -> (Output_Get:double)
   Set Alpha_isTargeting(execute=@n19.else, Alpha_isTargeting=0)
   n20 = float * float(A=GetRealTimeSeconds(), B=Get speedWaveHairBack() -> (speedWaveHairBack:double))
   n21 = float * float(A=MultiplyByPi(Value="Sin (Radians)"(A=@n20)), B=Get amplitudWaveHairback() -> (amplitudWaveHairback:double))
   n22 = Get offsetWaveHairBack() -> (offsetWaveHairBack:double)
   Set rotatorWaveHairBack(execute=@n15.then_5, rotatorWaveHairBack:Rotator=MakeRotator(Pitch=float + float(A=@n21, B=@n22))) -> (Output_Get:Rotator)
   n23 = Branch(execute=@n15.then_6, Condition=Get bIsSkydiveFloating())
   > Set Sim Space Settings(Sim Space Settings:Sim Space Settings Structure=Get Sim Space Settings_isFloatingTrue() -> (Sim Space Settings_isFloatingTrue:Sim Space Settings Structure)) -> (Output_Get:Sim Space Settings Structure)
   Set Sim Space Settings(
    execute=@n23.else
    Sim Space Settings=Get Sim Space Settings_isFloatingFalse() -> (Sim Space Settings_isFloatingFalse:Sim Space Settings Structure)
   )
   n24 = Branch(execute=@n15.then_0, Condition=Get bIsSprinting())
   > n25 = Branch(Condition=Get bIsTacticalSprinting())
   > Set speedWaveHairTop(speedWaveHairTop:double=18) -> (Output_Get:double)
   > Set amplitudWaveHairTop(amplitudWaveHairTop:double=5) -> (Output_Get:double)
   > Set offsetWaveHairTop(offsetWaveHairTop:double=5) -> (Output_Get:double)
   > Set speedWaveHairBack(speedWaveHairBack:double=16) -> (Output_Get:double)
   > Set amplitudWaveHairback(amplitudWaveHairback:double=3) -> (Output_Get:double)
   > Set offsetWaveHairBack(offsetWaveHairBack:double=10) -> (Output_Get:double)
   Set speedWaveHairTop(execute=@n25.else, speedWaveHairTop=13)
   > Set amplitudWaveHairTop(amplitudWaveHairTop=3)
   > Set offsetWaveHairTop(offsetWaveHairTop=0)
   > Set speedWaveHairBack(speedWaveHairBack=13)
   > Set amplitudWaveHairback(amplitudWaveHairback=2)
   > Set offsetWaveHairBack(offsetWaveHairBack=5)
   Set speedWaveHairTop(execute=@n24.else, speedWaveHairTop=0)
   > Set speedWaveHairBack(speedWaveHairBack=0)
   n26 = Get bIsMoving2D()
   n27 = Branch(execute=@n15.then_9, Condition=AND Boolean(A=Get bIsDBNO(), B=NOT Boolean(A=@n26)))
   > Set Alpha_isDBNOIdle(Alpha_isDBNOIdle:double=1) -> (Output_Get:double)
   Set Alpha_isDBNOIdle(execute=@n27.else, Alpha_isDBNOIdle=0)
   n28 = CheckForBackbling()
   Set isWearingBackbling(
    execute=@n15.then_10
    isWearingBackbling=AND Boolean!(A=@n28, B=NOT Boolean(A=@n28.bIsBackblingLong), C:bool=NOT Boolean(A=@n28.bIsBackblingLong)) -> (ReturnValue)
   )
   > Set Is Wearing Cape(Is Wearing Cape=OR Boolean(A=@n28.bIsBackblingACape, B=@n28.bIsBackblingLong))
   n29 = Branch(execute=@n15.then_11, Condition=AND Boolean!(A=Get bIsSkydiveFloating(), B=Get isWearingBackbling()) -> (ReturnValue))
   > Set Alpha_isUseBackpack(Alpha_isUseBackpack:double=1) -> (Output_Get:double)
   Set Alpha_isUseBackpack(execute=@n29.else, Alpha_isUseBackpack=0)
   n30 = Branch(execute=@n15.then_8, Condition=Get bIsDBNO())
   > Set Alpha_isDBNO(Alpha_isDBNO:double=1) -> (Output_Get:double)
   Set Alpha_isDBNO(execute=@n30.else, Alpha_isDBNO=0)
   n31 = Branch(execute=@n15.then_2)
   > Set Alpha_isDisableRigbody(Alpha_isDisableRigbody:double=1) -> (Output_Get:double)
   Set Alpha_isDisableRigbody(execute=@n31.else, Alpha_isDisableRigbody=0)
   n32 = Get bRigidBodyEnabled()
   n33 = Branch(execute=@n15.then_4, Condition=AND Boolean(A=Get bIsSprinting(), B=NOT Boolean(A=@n32)))
   > Set Alpha_isDisableRigbodySprinting(Alpha_isDisableRigbodySprinting:double=1) -> (Output_Get:double)
   Set Alpha_isDisableRigbodySprinting(execute=@n33.else, Alpha_isDisableRigbodySprinting=0)
   n34 = Branch(execute=@n15.then_7, Condition=Get bIsParachuteOpened())
   > Set Alpha_isOpenParachute(Alpha_isOpenParachute:double=1) -> (Output_Get:double)
   Set Alpha_isOpenParachute(execute=@n34.else, Alpha_isOpenParachute=0)
   n35 = Get bIsTacticalSprinting()
   n36 = Branch(execute=@n1.then_2, Condition=AND Boolean(A=NOT Boolean(A=@n35), B=Get bIsFalling()))
   > Set Alpha_isFalling(Alpha_isFalling:double=1) -> (Output_Get:double)
   Set Alpha_isFalling(execute=@n36.else, Alpha_isFalling=0)
   OR Boolean(A=false, B=false)
   Get bIsCrouching()
   Get bIsCrouchSprinting()
  ]}
]
variables [
 Alpha_isCrouching {type=number}
 Alpha_isDBNO {type=number}
 Alpha_isDBNOIdle {type=number}
 Alpha_isDisableRigbody {type=number}
 Alpha_isDisableRigbodySprinting {type=number}
 Alpha_isFalling {type=number}
 Alpha_isOpenParachute {type=number}
 Alpha_isSprinting {type=number}
 Alpha_isTargeting {type=number}
 Alpha_isUseBackpack {type=number}
 Alpha_isUseCtrl {type=number}
 Alpha_isUseCtrl_test {type=number}
 Alpha_moveSideCrouch {type=number}
 Alpha_SkyDiveFloating {type=number}
 Alpha_SkydiveMode {type=number}
 Alpha_sprintCrouch {type=number}
 amplitudWaveHairback {type=number, default=6}
 amplitudWaveHairTop {type=number, default=6}
 Gravity Override Params {
  type=GravityOverrideParamsStruct
  default {
   emoteOrMelee {gravityOverride=<0 0 -1200>}
   onGround_Falling {gravityOverride=<0 0 2000>, windAmplitude=<100 100 -250>, windFrequency=<5 5 20>}
   onGround_Moving {gravityOverride=<0 0 -800>, windAmplitude=<100 300 300>, windFrequency=<5 5 10>}
   onGround_Standing {gravityOverride=<0 0 -1200>}
   skydive_DivingDown {gravityOverride=<-600 0 0>, windAmplitude=<10 300 500>, windFrequency=<2 15 18>}
   skydive_FloatingDown {gravityOverride=<0 -400 400>, windFrequency=<4 15 16>, windOffset=<0 -300 200>}
   skydive_GoingUp {gravityOverride=<0 -150 700>, windAmplitude=<100 -250 200>}
   skydive_Parachuting {gravityOverride=<0 -200 -1500>, windAmplitude=<100 -250 200>, windFrequency=<3 14 16>}
   tactical_Sprint {gravityOverride=<0 -100 -600>, windAmplitude=<100 -250 250>, windFrequency=<5 10 15>}
   ziplining {gravityOverride=<0 0 -1200>}
  }
 }
 GravityOverride {type=Vector}
 Is Parachute Opened {type=boolean}
 Is Wearing Cape {type=boolean}
 isWearingBackbling {type=boolean}
 offsetWaveHairBack {type=number, default=10}
 offsetWaveHairTop {type=number, default=10}
 rotatorWaveHairBack {type=Rotator}
 rotatorWaveHairTop {type=Rotator}
 Sim Space Settings {
  type=SimSpaceSettings
  default {
   dampingAlpha=1
   maxAngularAcceleration=10000
   maxAngularVelocity=10000
   maxLinearAcceleration=10000
   maxLinearVelocity=10000
   velocityScaleZ=1
   worldAlpha=0
   externalLinearDragV=<0 0 0>
  }
 }
 Sim Space Settings_isFloatingFalse {
  type=SimSpaceSettings
  default {
   dampingAlpha=1
   maxAngularAcceleration=500
   maxAngularVelocity=5
   maxLinearAcceleration=300
   maxLinearVelocity=500
   velocityScaleZ=0.2
   worldAlpha=0.75
   externalLinearDragV=<0.2 0.2 0.2>
  }
 }
 Sim Space Settings_isFloatingTrue {
  type=SimSpaceSettings
  default {
   dampingAlpha=1
   maxAngularAcceleration=600
   maxAngularVelocity=3
   maxLinearAcceleration=500
   maxLinearVelocity=300
   velocityScaleZ=0.1
   worldAlpha=0.2
   externalLinearDragV=<0.2 0.2 0.2>
  }
 }
 speedWaveHairBack {type=number, default=15}
 speedWaveHairTop {type=number, default=15}
 SprintingNotJumping {type=boolean}
 TacticalSprintingNotJumping {type=boolean}
]
